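# Animation Blueprint for the F_MED_WonderHill female body cosmetic that adds procedural secondary motion to its jacket bones.
# AnimDynamics nodes run in series: two-bone chains for center, back and sides, then single bones for the front.
# Spring, gravity, angular, spherical and planar limits keep the jacket off the body.
# Two such stacks branch from cached 'BeforeJacketPose', the second with much tighter center and left-side angular limits.
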AnimBlueprint /BRCosmetics/Characters/Player/Female/Medium/Bodies/F_MED_WonderHill/Meshes/F_MED_WonderHill_AnimBP
target_skeleton=&/F_MED_WonderHill_Skeleton
graphs [
 AnimGraph {nodes [
   "Anim Dynamics - Chain: dyn_jacket_center_01 -> dyn_jacket_center_02"(Alpha=0.5, ExternalForce="0,0,0") {
    angularDampingOverride=0.999
    angularSpringConstant=300
    bAngularSpring=true
    bChain=true
    bOverrideAngularDamping=true
    bUseGravityOverride=true
    bUsePlanarLimit=false
    bUseSphericalLimits=true
    linearDampingOverride=0.8
    simulationSpace=RootRelative
    alphaBoolBlend {blendOption=Linear}
    boundBone {boneName=dyn_jacket_center_01}
    chainEnd {boneName=dyn_jacket_center_02}
    componentAppliedLinearAccClamp=<3000 3000 3000>
    componentLinearAccScale=<0.3 0.3 0.3>
    componentLinearVelScale=<0.3 0.3 0.3>
    gravityOverride {z=-980}
    physicsBodyDefinitions [
     {
      boundBone {boneName=dyn_jacket_center_01}
      boxExtents=<20 20 20>
      constraintSetup {angularLimitsMax=<0 -5 1>, angularLimitsMin=<0 -25 -1>, angularTarget=<0 0 0>}
      localJointOffset=<2 0 0>
     }
     {
      boundBone {boneName=dyn_jacket_center_02}
      boxExtents=<20 20 20>
      constraintSetup {angularLimitsMax=<0 0 0>, angularLimitsMin=<0 -15 0>, angularTarget=<0 0 0>}
      localJointOffset=<2 0 0>
     }
    ]
    relativeSpaceBone {boneName=root}
    sphericalLimits [
     {limitRadius=12, drivingBone {boneName=spine_02}, sphereLocalOffset=<0 2 0>}
     {limitRadius=12, drivingBone {boneName=spine_04}, sphereLocalOffset=<0 3 0>}
     {limitRadius=12, drivingBone {boneName=spine_03}, sphereLocalOffset=<0 3 0>}
     {limitRadius=12, drivingBone {boneName=spine_05}, sphereLocalOffset=<0 3 0>}
     {limitRadius=20, drivingBone {boneName=pelvis}, sphereLocalOffset=<0 0 0>}
    ]
   }
   "Anim Dynamics - Chain: dyn_jacket_center_01 -> dyn_jacket_center_02"(Alpha=0.5, ExternalForce="0,0,0") {
    angularDampingOverride=0.999
    angularSpringConstant=300
    bAngularSpring=true
    bChain=true
    bOverrideAngularDamping=true
    bUseGravityOverride=true
    bUsePlanarLimit=false
    bUseSphericalLimits=true
    linearDampingOverride=0.8
    simulationSpace=RootRelative
    alphaBoolBlend {blendOption=Linear}
    boundBone {boneName=dyn_jacket_center_01}
    chainEnd {boneName=dyn_jacket_center_02}
    componentAppliedLinearAccClamp=<3000 3000 3000>
    componentLinearAccScale=<0.3 0.3 0.3>
    componentLinearVelScale=<0.3 0.3 0.3>
    gravityOverride {z=-980}
    physicsBodyDefinitions [
     {
      boundBone {boneName=dyn_jacket_center_01}
      boxExtents=<20 20 20>
      constraintSetup {angularLimitsMax=<0 -5 1>, angularLimitsMin=<0 -25 -1>, angularTarget=<0 0 0>}
      localJointOffset=<2 0 0>
     }
     {
      boundBone {boneName=dyn_jacket_center_02}
      boxExtents=<20 20 20>
      constraintSetup {angularLimitsMax=<0 0 0>, angularLimitsMin=<0 -15 0>, angularTarget=<0 0 0>}
      localJointOffset=<2 0 0>
     }
    ]
    relativeSpaceBone {boneName=root}
    sphericalLimits [
     {limitRadius=12, drivingBone {boneName=spine_02}, sphereLocalOffset=<0 2 0>}
     {limitRadius=12, drivingBone {boneName=spine_04}, sphereLocalOffset=<0 3 0>}
     {limitRadius=12, drivingBone {boneName=spine_03}, sphereLocalOffset=<0 3 0>}
     {limitRadius=12, drivingBone {boneName=spine_05}, sphereLocalOffset=<0 3 0>}
     {limitRadius=20, drivingBone {boneName=pelvis}, sphereLocalOffset=<0 0 0>}
    ]
   }
   "Anim Dynamics - Chain: dyn_jacket_back_01_r -> dyn_jacket_back_02_r"(Alpha=1, ExternalForce="0,0,0") {
    angularDampingOverride=0.999
    angularSpringConstant=300
    bAngularSpring=true
    bChain=true
    bOverrideAngularDamping=true
    bUseGravityOverride=true
    linearDampingOverride=0.8
    simulationSpace=RootRelative
    alphaBoolBlend {blendOption=Linear}
    boundBone {boneName=dyn_jacket_back_01_r}
    chainEnd {boneName=dyn_jacket_back_02_r}
    componentAppliedLinearAccClamp=<3000 3000 3000>
    componentLinearAccScale=<0.3 0.3 0.3>
    componentLinearVelScale=<0.3 0.3 0.3>
    gravityOverride {z=-980}
    physicsBodyDefinitions [
     {
      boundBone {boneName=dyn_jacket_back_01_r}
      boxExtents=<20 20 20>
      constraintSetup {angularLimitsMax=<2 0 2>, angularLimitsMin=<-2 -25 -2>, angularTarget=<0 0 0>}
      localJointOffset=<2 0 0>
     }
     {
      boundBone {boneName=dyn_jacket_back_02_r}
      boxExtents=<20 20 20>
      constraintSetup {angularLimitsMax=<2 0 2>, angularLimitsMin=<-2 -15 -2>, angularTarget=<0 0 0>}
      localJointOffset=<2 0 0>
     }
    ]
    planarLimits [{drivingBone {boneName=spine_01}, planeTransform {location=<0 0 15>, rotation=<rot 0 -5 -45>}}]
    relativeSpaceBone {boneName=root}
    sphericalLimits [
     {limitRadius=12, drivingBone {boneName=spine_02}, sphereLocalOffset=<0 0 1>}
     {limitRadius=10, drivingBone {boneName=spine_04}, sphereLocalOffset=<0 0 1>}
     {limitRadius=10, drivingBone {boneName=spine_03}, sphereLocalOffset=<0 0 1>}
     {limitRadius=10, drivingBone {boneName=spine_05}, sphereLocalOffset=<0 0 1>}
     {limitRadius=20, drivingBone {boneName=pelvis}, sphereLocalOffset=<0 0 0>}
    ]
   }
   "Anim Dynamics - Chain: dyn_jacket_back_01_l -> dyn_jacket_back_02_l"(Alpha=1, ExternalForce="0,0,0") {
    angularDampingOverride=0.999
    angularSpringConstant=300
    bAngularSpring=true
    bChain=true
    bOverrideAngularDamping=true
    bUseGravityOverride=true
    linearDampingOverride=0.8
    simulationSpace=RootRelative
    alphaBoolBlend {blendOption=Linear}
    boundBone {boneName=dyn_jacket_back_01_l}
    chainEnd {boneName=dyn_jacket_back_02_l}
    componentAppliedLinearAccClamp=<3000 3000 3000>
    componentLinearAccScale=<0.3 0.3 0.3>
    componentLinearVelScale=<0.3 0.3 0.3>
    gravityOverride {z=-980}
    physicsBodyDefinitions [
     {
      boundBone {boneName=dyn_jacket_back_01_l}
      boxExtents=<20 20 20>
      constraintSetup {angularLimitsMax=<2 0 2>, angularLimitsMin=<-2 -25 -2>, angularTarget=<0 0 0>}
      localJointOffset=<3 0 0>
     }
     {
      boundBone {boneName=dyn_jacket_back_02_l}
      boxExtents=<20 20 20>
      constraintSetup {angularLimitsMax=<2 0 2>, angularLimitsMin=<-2 -15 -2>, angularTarget=<0 0 0>}
      localJointOffset=<3 0 0>
     }
    ]
    planarLimits [{drivingBone {boneName=spine_01}, planeTransform {location=<0 0 -13>, rotation=<rot 0 0 -150>}}]
    relativeSpaceBone {boneName=root}
    sphericalLimits [
     {limitRadius=10, drivingBone {boneName=spine_02}, sphereLocalOffset=<0 0 -1>}
     {limitRadius=10, drivingBone {boneName=spine_04}, sphereLocalOffset=<0 0 -1>}
     {limitRadius=10, drivingBone {boneName=spine_03}, sphereLocalOffset=<0 0 -1>}
     {limitRadius=10, drivingBone {boneName=spine_05}, sphereLocalOffset=<0 0 -1>}
     {limitRadius=20, drivingBone {boneName=pelvis}, sphereLocalOffset=<0 0 0>}
    ]
   }
   "Anim Dynamics - Chain: dyn_jacket_side_01_l -> dyn_jacket_side_02_l"(Alpha=1, ExternalForce="0,0,0") {
    angularDampingOverride=0.999
    angularSpringConstant=300
    bAngularSpring=true
    bChain=true
    bOverrideAngularDamping=true
    bUseGravityOverride=true
    simulationSpace=RootRelative
    alphaBoolBlend {blendOption=Linear}
    boundBone {boneName=dyn_jacket_side_01_l}
    chainEnd {boneName=dyn_jacket_side_02_l}
    componentAppliedLinearAccClamp=<3000 3000 3000>
    componentLinearAccScale=<0.3 0.3 0.3>
    componentLinearVelScale=<0.3 0.3 0.3>
    gravityOverride {z=-980}
    physicsBodyDefinitions [
     {
      boundBone {boneName=dyn_jacket_side_01_l}
      boxExtents=<20 10 20>
      constraintSetup {angularLimitsMax=<2 5 2>, angularLimitsMin=<-2 -25 -2>, angularTarget=<0 0 0>}
      localJointOffset=<3 0 0>
     }
     {
      boundBone {boneName=dyn_jacket_side_02_l}
      boxExtents=<20 20 20>
      constraintSetup {angularLimitsMax=<2 5 2>, angularLimitsMin=<-2 -15 -2>, angularTarget=<0 0 0>}
      localJointOffset=<3 0 0>
     }
    ]
    planarLimits [{drivingBone {boneName=spine_01}, planeTransform {location=<0 0 -11>, rotation=<rot 0 0 -180>}}]
    relativeSpaceBone {boneName=root}
    sphericalLimits [
     {limitRadius=10, drivingBone {boneName=spine_02}, sphereLocalOffset=<0 0 -2>}
     {limitRadius=10, drivingBone {boneName=spine_04}, sphereLocalOffset=<0 0 -2>}
     {limitRadius=10, drivingBone {boneName=spine_03}, sphereLocalOffset=<0 0 -2>}
     {limitRadius=10, drivingBone {boneName=spine_05}, sphereLocalOffset=<0 0 -2>}
    ]
   }
   "Anim Dynamics - Chain: dyn_jacket_side_01_r -> dyn_jacket_side_02_r"(Alpha=1, ExternalForce="0,0,0") {
    angularDampingOverride=0.999
    angularSpringConstant=300
    bAngularSpring=true
    bChain=true
    bOverrideAngularDamping=true
    bUseGravityOverride=true
    simulationSpace=RootRelative
    alphaBoolBlend {blendOption=Linear}
    boundBone {boneName=dyn_jacket_side_01_r}
    chainEnd {boneName=dyn_jacket_side_02_r}
    componentAppliedLinearAccClamp=<3000 3000 3000>
    componentLinearAccScale=<0.3 0.3 0.3>
    componentLinearVelScale=<0.3 0.3 0.3>
    gravityOverride {z=-980}
    physicsBodyDefinitions [
     {
      boundBone {boneName=dyn_jacket_side_01_r}
      boxExtents=<20 10 20>
      constraintSetup {angularLimitsMax=<2 0 2>, angularLimitsMin=<-2 -25 -2>, angularTarget=<0 0 0>}
      localJointOffset=<3 0 0>
     }
     {
      boundBone {boneName=dyn_jacket_side_02_r}
      boxExtents=<20 20 20>
      constraintSetup {angularLimitsMax=<2 5 2>, angularLimitsMin=<-2 -15 -2>, angularTarget=<0 0 0>}
      localJointOffset=<3 0 0>
     }
    ]
    planarLimits [{drivingBone {boneName=spine_01}, planeTransform {location=<0 0 12>, rotation=<rot 0 0 0>}}]
    relativeSpaceBone {boneName=root}
    sphericalLimits [
     {limitRadius=10, drivingBone {boneName=spine_02}, sphereLocalOffset=<0 0 -2>}
     {limitRadius=10, drivingBone {boneName=spine_04}, sphereLocalOffset=<0 0 -2>}
     {limitRadius=10, drivingBone {boneName=spine_03}, sphereLocalOffset=<0 0 -2>}
     {limitRadius=10, drivingBone {boneName=spine_05}, sphereLocalOffset=<0 0 -2>}
     {limitRadius=10, drivingBone {boneName=lowerarm_r}, sphereLocalOffset=<0 0 0>}
    ]
   }
   n1 = "Anim Dynamics - Chain: dyn_jacket_center_01 -> dyn_jacket_center_02"(ComponentPose=Local To Component(LocalPose=Use cached pose 'BeforeJacketPose'()), Alpha=1, ExternalForce="0,0,0") {
    angularDampingOverride=0.999
    angularSpringConstant=300
    bAngularSpring=true
    bChain=true
    bOverrideAngularDamping=true
    bUseGravityOverride=true
    linearDampingOverride=0.8
    simulationSpace=RootRelative
    alphaBoolBlend {blendOption=Linear}
    boundBone {boneName=dyn_jacket_center_01}
    chainEnd {boneName=dyn_jacket_center_02}
    componentAppliedLinearAccClamp=<3000 3000 3000>
    componentLinearAccScale=<0.3 0.3 0.3>
    componentLinearVelScale=<0.3 0.3 0.3>
    gravityOverride {z=-980}
    physicsBodyDefinitions [
     {
      boundBone {boneName=dyn_jacket_center_01}
      boxExtents=<20 20 20>
      constraintSetup {angularLimitsMax=<2 -5 2>, angularLimitsMin=<-2 -25 -2>, angularTarget=<0 0 0>}
      localJointOffset=<2 0 0>
     }
     {
      boundBone {boneName=dyn_jacket_center_02}
      boxExtents=<20 20 20>
      constraintSetup {angularLimitsMax=<2 0 2>, angularLimitsMin=<-2 -15 -2>, angularTarget=<0 0 0>}
      localJointOffset=<2 0 0>
     }
    ]
    planarLimits [{drivingBone {boneName=spine_01}, planeTransform {location=<0 -14 0>, rotation=<rot 0 30 -90>}}]
    relativeSpaceBone {boneName=root}
    sphericalLimits [
     {limitRadius=10, drivingBone {boneName=spine_02}, sphereLocalOffset=<0 2 0>}
     {limitRadius=10, drivingBone {boneName=spine_04}, sphereLocalOffset=<0 3 0>}
     {limitRadius=10, drivingBone {boneName=spine_03}, sphereLocalOffset=<0 3 0>}
     {limitRadius=10, drivingBone {boneName=spine_05}, sphereLocalOffset=<0 3 0>}
     {limitRadius=20, drivingBone {boneName=pelvis}, sphereLocalOffset=<0 0 0>}
    ]
   }
   n2 = "Anim Dynamics - Chain: dyn_jacket_back_01_l -> dyn_jacket_back_02_l"(ComponentPose=@n1, Alpha=1, ExternalForce="0,0,0") {
    angularDampingOverride=0.999
    angularSpringConstant=300
    bAngularSpring=true
    bChain=true
    bOverrideAngularDamping=true
    bUseGravityOverride=true
    linearDampingOverride=0.8
    simulationSpace=RootRelative
    alphaBoolBlend {blendOption=Linear}
    boundBone {boneName=dyn_jacket_back_01_l}
    chainEnd {boneName=dyn_jacket_back_02_l}
    componentAppliedLinearAccClamp=<3000 3000 3000>
    componentLinearAccScale=<0.3 0.3 0.3>
    componentLinearVelScale=<0.3 0.3 0.3>
    gravityOverride {z=-980}
    physicsBodyDefinitions [
     {
      boundBone {boneName=dyn_jacket_back_01_l}
      boxExtents=<20 20 20>
      constraintSetup {angularLimitsMax=<2 0 2>, angularLimitsMin=<-2 -25 -2>, angularTarget=<0 0 0>}
      localJointOffset=<3 0 0>
     }
     {
      boundBone {boneName=dyn_jacket_back_02_l}
      boxExtents=<20 20 20>
      constraintSetup {angularLimitsMax=<2 0 2>, angularLimitsMin=<-2 -15 -2>, angularTarget=<0 0 0>}
      localJointOffset=<3 0 0>
     }
    ]
    planarLimits [{drivingBone {boneName=spine_01}, planeTransform {location=<0 0 -10>, rotation=<rot 0 0 -150>}}]
    relativeSpaceBone {boneName=root}
    sphericalLimits [
     {limitRadius=10, drivingBone {boneName=spine_02}, sphereLocalOffset=<0 0 -1>}
     {limitRadius=10, drivingBone {boneName=spine_04}, sphereLocalOffset=<0 0 -1>}
     {limitRadius=10, drivingBone {boneName=spine_03}, sphereLocalOffset=<0 0 -1>}
     {limitRadius=10, drivingBone {boneName=spine_05}, sphereLocalOffset=<0 0 -1>}
     {limitRadius=20, drivingBone {boneName=pelvis}, sphereLocalOffset=<0 0 0>}
    ]
   }
   n3 = "Anim Dynamics - Chain: dyn_jacket_side_01_l -> dyn_jacket_side_02_l"(ComponentPose=@n2, Alpha=1, ExternalForce="0,0,0") {
    angularDampingOverride=0.999
    angularSpringConstant=300
    bAngularSpring=true
    bChain=true
    bOverrideAngularDamping=true
    bUseGravityOverride=true
    simulationSpace=RootRelative
    alphaBoolBlend {blendOption=Linear}
    boundBone {boneName=dyn_jacket_side_01_l}
    chainEnd {boneName=dyn_jacket_side_02_l}
    componentAppliedLinearAccClamp=<3000 3000 3000>
    componentLinearAccScale=<0.3 0.3 0.3>
    componentLinearVelScale=<0.3 0.3 0.3>
    gravityOverride {z=-980}
    physicsBodyDefinitions [
     {
      boundBone {boneName=dyn_jacket_side_01_l}
      boxExtents=<20 10 20>
      constraintSetup {angularLimitsMax=<2 5 2>, angularLimitsMin=<-2 -25 -2>, angularTarget=<0 0 0>}
      localJointOffset=<3 0 0>
     }
     {
      boundBone {boneName=dyn_jacket_side_02_l}
      boxExtents=<20 20 20>
      constraintSetup {angularLimitsMax=<2 5 2>, angularLimitsMin=<-2 -15 -2>, angularTarget=<0 0 0>}
      localJointOffset=<3 0 0>
     }
    ]
    planarLimits [{drivingBone {boneName=spine_01}, planeTransform {location=<0 0 -10>, rotation=<rot 0 0 -180>}}]
    relativeSpaceBone {boneName=root}
    sphericalLimits [
     {limitRadius=10, drivingBone {boneName=spine_02}, sphereLocalOffset=<0 0 -2>}
     {limitRadius=10, drivingBone {boneName=spine_04}, sphereLocalOffset=<0 0 -2>}
     {limitRadius=10, drivingBone {boneName=spine_03}, sphereLocalOffset=<0 0 -2>}
     {limitRadius=10, drivingBone {boneName=spine_05}, sphereLocalOffset=<0 0 -2>}
    ]
   }
   n4 = "Anim Dynamics - Chain: dyn_jacket_back_01_r -> dyn_jacket_back_02_r"(ComponentPose=@n3, Alpha=1, ExternalForce="0,0,0") {
    angularDampingOverride=0.999
    angularSpringConstant=300
    bAngularSpring=true
    bChain=true
    bOverrideAngularDamping=true
    bUseGravityOverride=true
    linearDampingOverride=0.8
    simulationSpace=RootRelative
    alphaBoolBlend {blendOption=Linear}
    boundBone {boneName=dyn_jacket_back_01_r}
    chainEnd {boneName=dyn_jacket_back_02_r}
    componentAppliedLinearAccClamp=<3000 3000 3000>
    componentLinearAccScale=<0.3 0.3 0.3>
    componentLinearVelScale=<0.3 0.3 0.3>
    gravityOverride {z=-980}
    physicsBodyDefinitions [
     {
      boundBone {boneName=dyn_jacket_back_01_r}
      boxExtents=<20 20 20>
      constraintSetup {angularLimitsMax=<2 0 2>, angularLimitsMin=<-2 -25 -2>, angularTarget=<0 0 0>}
      localJointOffset=<2 0 0>
     }
     {
      boundBone {boneName=dyn_jacket_back_02_r}
      boxExtents=<20 20 20>
      constraintSetup {angularLimitsMax=<2 0 2>, angularLimitsMin=<-2 -15 -2>, angularTarget=<0 0 0>}
      localJointOffset=<2 0 0>
     }
    ]
    planarLimits [
     {drivingBone {boneName=spine_01}, planeTransform {location=<0 0 12>, rotation=<rot 0 0 -40>}}
     {drivingBone {boneName=pelvis}, planeTransform {location=<0 -10 0>, rotation=<rot 0 0 -90>}}
    ]
    relativeSpaceBone {boneName=root}
    sphericalLimits [
     {limitRadius=12, drivingBone {boneName=spine_02}, sphereLocalOffset=<0 0 1>}
     {limitRadius=10, drivingBone {boneName=spine_04}, sphereLocalOffset=<0 0 1>}
     {limitRadius=10, drivingBone {boneName=spine_03}, sphereLocalOffset=<0 0 1>}
     {limitRadius=10, drivingBone {boneName=spine_05}, sphereLocalOffset=<0 0 1>}
     {limitRadius=20, drivingBone {boneName=pelvis}, sphereLocalOffset=<0 0 0>}
    ]
   }
   n5 = "Anim Dynamics - Chain: dyn_jacket_side_01_r -> dyn_jacket_side_02_r"(ComponentPose=@n4, Alpha=1, ExternalForce="0,0,0") {
    angularDampingOverride=0.999
    angularSpringConstant=300
    bAngularSpring=true
    bChain=true
    bOverrideAngularDamping=true
    bUseGravityOverride=true
    simulationSpace=RootRelative
    alphaBoolBlend {blendOption=Linear}
    boundBone {boneName=dyn_jacket_side_01_r}
    chainEnd {boneName=dyn_jacket_side_02_r}
    componentAppliedLinearAccClamp=<3000 3000 3000>
    componentLinearAccScale=<0.3 0.3 0.3>
    componentLinearVelScale=<0.3 0.3 0.3>
    gravityOverride {z=-980}
    physicsBodyDefinitions [
     {
      boundBone {boneName=dyn_jacket_side_01_r}
      boxExtents=<20 10 20>
      constraintSetup {angularLimitsMax=<2 0 2>, angularLimitsMin=<-2 -25 -2>, angularTarget=<0 0 0>}
      localJointOffset=<3 0 0>
     }
     {
      boundBone {boneName=dyn_jacket_side_02_r}
      boxExtents=<20 20 20>
      constraintSetup {angularLimitsMax=<2 5 2>, angularLimitsMin=<-2 -15 -2>, angularTarget=<0 0 0>}
      localJointOffset=<3 0 0>
     }
    ]
    planarLimits [{drivingBone {boneName=spine_01}, planeTransform {location=<0 0 11>, rotation=<rot 0 0 0>}}]
    relativeSpaceBone {boneName=root}
    sphericalLimits [
     {limitRadius=10, drivingBone {boneName=spine_02}, sphereLocalOffset=<0 0 -2>}
     {limitRadius=10, drivingBone {boneName=spine_04}, sphereLocalOffset=<0 0 -2>}
     {limitRadius=10, drivingBone {boneName=spine_03}, sphereLocalOffset=<0 0 -2>}
     {limitRadius=10, drivingBone {boneName=spine_05}, sphereLocalOffset=<0 0 -2>}
     {limitRadius=10, drivingBone {boneName=lowerarm_r}, sphereLocalOffset=<0 0 0>}
    ]
   }
   n6 = Anim Dynamics - Bone: dyn_jacket_front_01_r(ComponentPose=@n5, Alpha=1, ExternalForce="0,0,0") {
    angularDampingOverride=0.7
    angularSpringConstant=500
    bAngularSpring=true
    bOverrideAngularDamping=true
    bUseGravityOverride=true
    bUsePlanarLimit=false
    bUseSphericalLimits=true
    simulationSpace=RootRelative
    alphaBoolBlend {blendOption=Linear}
    boundBone {boneName=dyn_jacket_front_01_r}
    componentAppliedLinearAccClamp=<3000 3000 3000>
    componentLinearAccScale=<0.3 0.3 0.3>
    componentLinearVelScale=<0.3 0.3 0.3>
    gravityOverride {z=-980}
    physicsBodyDefinitions [{
      boundBone {boneName=dyn_jacket_front_01_r}
      boxExtents=<30 30 30>
      constraintSetup {angularLimitsMax=<10 0 5>, angularLimitsMin=<-5 -30 -5>, angularTarget=<0 0 0>}
      localJointOffset=<1 0 0>
     }]
    relativeSpaceBone {boneName=root}
    sphericalLimits [
     {limitRadius=10, drivingBone {boneName=spine_02}, sphereLocalOffset=<0 0 0>}
     {limitRadius=10, drivingBone {boneName=spine_04}, sphereLocalOffset=<0 0 0>}
     {limitRadius=10, drivingBone {boneName=spine_03}, sphereLocalOffset=<0 0 0>}
     {limitRadius=10, drivingBone {boneName=spine_05}, sphereLocalOffset=<0 0 0>}
     {limitRadius=2, drivingBone {boneName=hand_r}, sphereLocalOffset=<0 0 0>}
    ]
   }
   n7 = Anim Dynamics - Bone: dyn_jacket_front_02_r(ComponentPose=@n6, Alpha=1, ExternalForce="0,0,0") {
    angularDampingOverride=0.7
    angularSpringConstant=500
    bAngularSpring=true
    bOverrideAngularDamping=true
    bUseGravityOverride=true
    bUsePlanarLimit=false
    bUseSphericalLimits=true
    simulationSpace=RootRelative
    alphaBoolBlend {blendOption=Linear}
    boundBone {boneName=dyn_jacket_front_02_r}
    componentAppliedLinearAccClamp=<3000 3000 3000>
    componentLinearAccScale=<0.3 0.3 0.3>
    componentLinearVelScale=<0.3 0.3 0.3>
    gravityOverride {z=-980}
    physicsBodyDefinitions [{
      boundBone {boneName=dyn_jacket_front_02_r}
      boxExtents=<30 30 30>
      constraintSetup {angularLimitsMax=<10 5 10>, angularLimitsMin=<-5 -20 -5>, angularTarget=<0 0 0>}
      localJointOffset=<1 0 0>
     }]
    relativeSpaceBone {boneName=root}
    sphericalLimits [
     {limitRadius=10, drivingBone {boneName=spine_02}, sphereLocalOffset=<0 0 0>}
     {limitRadius=10, drivingBone {boneName=spine_04}, sphereLocalOffset=<0 0 0>}
     {limitRadius=10, drivingBone {boneName=spine_03}, sphereLocalOffset=<0 0 0>}
     {limitRadius=10, drivingBone {boneName=spine_05}, sphereLocalOffset=<0 0 0>}
     {limitRadius=2, drivingBone {boneName=hand_r}, sphereLocalOffset=<0 0 0>}
    ]
   }
   n8 = Anim Dynamics - Bone: dyn_jacket_front_03_r(ComponentPose=@n7, Alpha=1, ExternalForce="0,0,0") {
    angularDampingOverride=0.7
    angularSpringConstant=500
    bAngularSpring=true
    bOverrideAngularDamping=true
    bUseGravityOverride=true
    bUsePlanarLimit=false
    bUseSphericalLimits=true
    simulationSpace=RootRelative
    alphaBoolBlend {blendOption=Linear}
    boundBone {boneName=dyn_jacket_front_03_r}
    componentAppliedLinearAccClamp=<3000 3000 3000>
    componentLinearAccScale=<0.3 0.3 0.3>
    componentLinearVelScale=<0.3 0.3 0.3>
    gravityOverride {z=-980}
    physicsBodyDefinitions [{
      boundBone {boneName=dyn_jacket_front_03_r}
      boxExtents=<30 30 30>
      constraintSetup {angularLimitsMax=<10 5 10>, angularLimitsMin=<-5 -30 -5>, angularTarget=<0 0 0>}
      localJointOffset=<1 0 0>
     }]
    relativeSpaceBone {boneName=root}
    sphericalLimits [
     {limitRadius=10, drivingBone {boneName=spine_02}, sphereLocalOffset=<0 0 0>}
     {limitRadius=10, drivingBone {boneName=spine_04}, sphereLocalOffset=<0 0 0>}
     {limitRadius=10, drivingBone {boneName=spine_03}, sphereLocalOffset=<0 0 0>}
     {limitRadius=10, drivingBone {boneName=spine_05}, sphereLocalOffset=<0 0 0>}
     {limitRadius=2, drivingBone {boneName=hand_r}, sphereLocalOffset=<0 0 0>}
    ]
   }
   n9 = Anim Dynamics - Bone: dyn_jacket_front_01_l(ComponentPose=@n8, Alpha=1, ExternalForce="0,0,0") {
    angularDampingOverride=0.7
    angularSpringConstant=500
    bAngularSpring=true
    bOverrideAngularDamping=true
    bUseGravityOverride=true
    bUsePlanarLimit=false
    bUseSphericalLimits=true
    simulationSpace=RootRelative
    alphaBoolBlend {blendOption=Linear}
    boundBone {boneName=dyn_jacket_front_01_l}
    componentAppliedLinearAccClamp=<3000 3000 3000>
    componentLinearAccScale=<0.3 0.3 0.3>
    componentLinearVelScale=<0.3 0.3 0.3>
    gravityOverride {z=-980}
    physicsBodyDefinitions [{
      boundBone {boneName=dyn_jacket_front_01_l}
      boxExtents=<30 30 30>
      constraintSetup {angularLimitsMax=<10 0 5>, angularLimitsMin=<-5 -30 -5>, angularTarget=<0 0 0>}
      localJointOffset=<1 0 0>
     }]
    relativeSpaceBone {boneName=root}
    sphericalLimits [
     {limitRadius=10, drivingBone {boneName=spine_02}, sphereLocalOffset=<0 0 0>}
     {limitRadius=10, drivingBone {boneName=spine_04}, sphereLocalOffset=<0 0 0>}
     {limitRadius=10, drivingBone {boneName=spine_03}, sphereLocalOffset=<0 0 0>}
     {limitRadius=10, drivingBone {boneName=spine_05}, sphereLocalOffset=<0 0 0>}
     {limitRadius=2, drivingBone {boneName=hand_r}, sphereLocalOffset=<0 0 0>}
    ]
   }
   n10 = Anim Dynamics - Bone: dyn_jacket_front_02_l(ComponentPose=@n9, Alpha=1, ExternalForce="0,0,0") {
    angularDampingOverride=0.7
    angularSpringConstant=500
    bAngularSpring=true
    bOverrideAngularDamping=true
    bUseGravityOverride=true
    bUsePlanarLimit=false
    bUseSphericalLimits=true
    simulationSpace=RootRelative
    alphaBoolBlend {blendOption=Linear}
    boundBone {boneName=dyn_jacket_front_02_l}
    componentAppliedLinearAccClamp=<3000 3000 3000>
    componentLinearAccScale=<0.3 0.3 0.3>
    componentLinearVelScale=<0.3 0.3 0.3>
    gravityOverride {z=-980}
    physicsBodyDefinitions [{
      boundBone {boneName=dyn_jacket_front_02_l}
      boxExtents=<30 30 30>
      constraintSetup {angularLimitsMax=<10 5 10>, angularLimitsMin=<-5 -20 -5>, angularTarget=<0 0 0>}
      localJointOffset=<1 0 0>
     }]
    relativeSpaceBone {boneName=root}
    sphericalLimits [
     {limitRadius=10, drivingBone {boneName=spine_02}, sphereLocalOffset=<0 0 0>}
     {limitRadius=10, drivingBone {boneName=spine_04}, sphereLocalOffset=<0 0 0>}
     {limitRadius=10, drivingBone {boneName=spine_03}, sphereLocalOffset=<0 0 0>}
     {limitRadius=10, drivingBone {boneName=spine_05}, sphereLocalOffset=<0 0 0>}
     {limitRadius=2, drivingBone {boneName=hand_r}, sphereLocalOffset=<0 0 0>}
    ]
   }
   n11 = Anim Dynamics - Bone: dyn_jacket_front_03_l(ComponentPose=@n10, Alpha=1, ExternalForce="0,0,0") {
    angularDampingOverride=0.7
    angularSpringConstant=500
    bAngularSpring=true
    bOverrideAngularDamping=true
    bUseGravityOverride=true
    bUsePlanarLimit=false
    bUseSphericalLimits=true
    simulationSpace=RootRelative
    alphaBoolBlend {blendOption=Linear}
    boundBone {boneName=dyn_jacket_front_03_l}
    componentAppliedLinearAccClamp=<3000 3000 3000>
    componentLinearAccScale=<0.3 0.3 0.3>
    componentLinearVelScale=<0.3 0.3 0.3>
    gravityOverride {z=-980}
    physicsBodyDefinitions [{
      boundBone {boneName=dyn_jacket_front_03_l}
      boxExtents=<30 30 30>
      constraintSetup {angularLimitsMax=<10 5 10>, angularLimitsMin=<-5 -30 -5>, angularTarget=<0 0 0>}
      localJointOffset=<1 0 0>
     }]
    relativeSpaceBone {boneName=root}
    sphericalLimits [
     {limitRadius=10, drivingBone {boneName=spine_02}, sphereLocalOffset=<0 0 0>}
     {limitRadius=10, drivingBone {boneName=spine_04}, sphereLocalOffset=<0 0 0>}
     {limitRadius=10, drivingBone {boneName=spine_03}, sphereLocalOffset=<0 0 0>}
     {limitRadius=10, drivingBone {boneName=spine_05}, sphereLocalOffset=<0 0 0>}
     {limitRadius=2, drivingBone {boneName=hand_r}, sphereLocalOffset=<0 0 0>}
    ]
   }
   n12 = "Anim Dynamics - Chain: dyn_jacket_center_01 -> dyn_jacket_center_02"(ComponentPose=Local To Component(LocalPose=Use cached pose 'BeforeJacketPose'()), Alpha=1, ExternalForce="0,0,0") {
    angularDampingOverride=0.999
    angularSpringConstant=300
    bAngularSpring=true
    bChain=true
    bOverrideAngularDamping=true
    bUseGravityOverride=true
    linearDampingOverride=0.8
    simulationSpace=RootRelative
    alphaBoolBlend {blendOption=Linear}
    boundBone {boneName=dyn_jacket_center_01}
    chainEnd {boneName=dyn_jacket_center_02}
    componentAppliedLinearAccClamp=<3000 3000 3000>
    componentLinearAccScale=<0.3 0.3 0.3>
    componentLinearVelScale=<0.3 0.3 0.3>
    gravityOverride {z=-980}
    physicsBodyDefinitions [
     {
      boundBone {boneName=dyn_jacket_center_01}
      boxExtents=<20 20 20>
      constraintSetup {angularLimitsMax=<0 -5 0>, angularLimitsMin=<0 -6 0>, angularTarget=<0 0 0>}
      localJointOffset=<2 0 0>
     }
     {
      boundBone {boneName=dyn_jacket_center_02}
      boxExtents=<20 20 20>
      constraintSetup {angularLimitsMax=<2 0 2>, angularLimitsMin=<0 -1 0>, angularTarget=<0 0 0>}
      localJointOffset=<2 0 0>
     }
    ]
    planarLimits [{drivingBone {boneName=spine_01}, planeTransform {location=<0 -14 0>, rotation=<rot 0 30.0001 -90.0001>}}]
    relativeSpaceBone {boneName=root}
    sphericalLimits [
     {limitRadius=10, drivingBone {boneName=spine_02}, sphereLocalOffset=<0 2 0>}
     {limitRadius=10, drivingBone {boneName=spine_04}, sphereLocalOffset=<0 3 0>}
     {limitRadius=10, drivingBone {boneName=spine_03}, sphereLocalOffset=<0 3 0>}
     {limitRadius=10, drivingBone {boneName=spine_05}, sphereLocalOffset=<0 3 0>}
     {limitRadius=20, drivingBone {boneName=pelvis}, sphereLocalOffset=<0 0 0>}
    ]
   }
   n13 = "Anim Dynamics - Chain: dyn_jacket_back_01_l -> dyn_jacket_back_02_l"(ComponentPose=@n12, Alpha=1, ExternalForce="0,0,0") {
    angularDampingOverride=0.999
    angularSpringConstant=300
    bAngularSpring=true
    bChain=true
    bOverrideAngularDamping=true
    bUseGravityOverride=true
    linearDampingOverride=0.8
    simulationSpace=RootRelative
    alphaBoolBlend {blendOption=Linear}
    boundBone {boneName=dyn_jacket_back_01_l}
    chainEnd {boneName=dyn_jacket_back_02_l}
    componentAppliedLinearAccClamp=<3000 3000 3000>
    componentLinearAccScale=<0.3 0.3 0.3>
    componentLinearVelScale=<0.3 0.3 0.3>
    gravityOverride {z=-980}
    physicsBodyDefinitions [
     {
      boundBone {boneName=dyn_jacket_back_01_l}
      boxExtents=<20 20 20>
      constraintSetup {angularLimitsMax=<0 -5 0>, angularLimitsMin=<0 -6 0>, angularTarget=<0 0 0>}
      localJointOffset=<3 0 0>
     }
     {
      boundBone {boneName=dyn_jacket_back_02_l}
      boxExtents=<20 20 20>
      constraintSetup {angularLimitsMax=<2 0 2>, angularLimitsMin=<0 -1 0>, angularTarget=<0 0 0>}
      localJointOffset=<3 0 0>
     }
    ]
    planarLimits [{drivingBone {boneName=spine_01}, planeTransform {location=<0 0 -13>, rotation=<rot 0 0 -150>}}]
    relativeSpaceBone {boneName=root}
    sphericalLimits [
     {limitRadius=10, drivingBone {boneName=spine_02}, sphereLocalOffset=<0 0 -1>}
     {limitRadius=10, drivingBone {boneName=spine_04}, sphereLocalOffset=<0 0 -1>}
     {limitRadius=10, drivingBone {boneName=spine_03}, sphereLocalOffset=<0 0 -1>}
     {limitRadius=10, drivingBone {boneName=spine_05}, sphereLocalOffset=<0 0 -1>}
     {limitRadius=20, drivingBone {boneName=pelvis}, sphereLocalOffset=<0 0 0>}
    ]
   }
   n14 = "Anim Dynamics - Chain: dyn_jacket_side_01_l -> dyn_jacket_side_02_l"(ComponentPose=@n13, Alpha=1, ExternalForce="0,0,0") {
    angularDampingOverride=0.999
    angularSpringConstant=300
    bAngularSpring=true
    bChain=true
    bOverrideAngularDamping=true
    bUseGravityOverride=true
    simulationSpace=RootRelative
    alphaBoolBlend {blendOption=Linear}
    boundBone {boneName=dyn_jacket_side_01_l}
    chainEnd {boneName=dyn_jacket_side_02_l}
    componentAppliedLinearAccClamp=<3000 3000 3000>
    componentLinearAccScale=<0.3 0.3 0.3>
    componentLinearVelScale=<0.3 0.3 0.3>
    gravityOverride {z=-980}
    physicsBodyDefinitions [
     {
      boundBone {boneName=dyn_jacket_side_01_l}
      boxExtents=<20 10 20>
      constraintSetup {angularLimitsMax=<0 -5 0>, angularLimitsMin=<0 -6 0>, angularTarget=<0 0 0>}
      localJointOffset=<3 0 0>
     }
     {
      boundBone {boneName=dyn_jacket_side_02_l}
      boxExtents=<20 20 20>
      constraintSetup {angularLimitsMax=<2 5 2>, angularLimitsMin=<0 -1 0>, angularTarget=<0 0 0>}
      localJointOffset=<3 0 0>
     }
    ]
    planarLimits [{drivingBone {boneName=spine_01}, planeTransform {location=<0 0 -11>, rotation=<rot 0 0 -180>}}]
    relativeSpaceBone {boneName=root}
    sphericalLimits [
     {limitRadius=10, drivingBone {boneName=spine_02}, sphereLocalOffset=<0 0 -2>}
     {limitRadius=10, drivingBone {boneName=spine_04}, sphereLocalOffset=<0 0 -2>}
     {limitRadius=10, drivingBone {boneName=spine_03}, sphereLocalOffset=<0 0 -2>}
     {limitRadius=10, drivingBone {boneName=spine_05}, sphereLocalOffset=<0 0 -2>}
    ]
   }
   n15 = "Anim Dynamics - Chain: dyn_jacket_back_01_r -> dyn_jacket_back_02_r"(ComponentPose=@n14, Alpha=1, ExternalForce="0,0,0") {
    angularDampingOverride=0.999
    angularSpringConstant=300
    bAngularSpring=true
    bChain=true
    bOverrideAngularDamping=true
    bUseGravityOverride=true
    linearDampingOverride=0.8
    simulationSpace=RootRelative
    alphaBoolBlend {blendOption=Linear}
    boundBone {boneName=dyn_jacket_back_01_r}
    chainEnd {boneName=dyn_jacket_back_02_r}
    componentAppliedLinearAccClamp=<3000 3000 3000>
    componentLinearAccScale=<0.3 0.3 0.3>
    componentLinearVelScale=<0.3 0.3 0.3>
    gravityOverride {z=-980}
    physicsBodyDefinitions [
     {
      boundBone {boneName=dyn_jacket_back_01_r}
      boxExtents=<20 20 20>
      constraintSetup {angularLimitsMax=<2 0 2>, angularLimitsMin=<-2 -25 -2>, angularTarget=<0 0 0>}
      localJointOffset=<2 0 0>
     }
     {
      boundBone {boneName=dyn_jacket_back_02_r}
      boxExtents=<20 20 20>
      constraintSetup {angularLimitsMax=<2 0 2>, angularLimitsMin=<-2 -15 -2>, angularTarget=<0 0 0>}
      localJointOffset=<2 0 0>
     }
    ]
    planarLimits [{drivingBone {boneName=spine_01}, planeTransform {location=<0 0 15>, rotation=<rot 0 -5 -45>}}]
    relativeSpaceBone {boneName=root}
    sphericalLimits [
     {limitRadius=12, drivingBone {boneName=spine_02}, sphereLocalOffset=<0 0 1>}
     {limitRadius=10, drivingBone {boneName=spine_04}, sphereLocalOffset=<0 0 1>}
     {limitRadius=10, drivingBone {boneName=spine_03}, sphereLocalOffset=<0 0 1>}
     {limitRadius=10, drivingBone {boneName=spine_05}, sphereLocalOffset=<0 0 1>}
     {limitRadius=20, drivingBone {boneName=pelvis}, sphereLocalOffset=<0 0 0>}
    ]
   }
   n16 = "Anim Dynamics - Chain: dyn_jacket_side_01_r -> dyn_jacket_side_02_r"(ComponentPose=@n15, Alpha=1, ExternalForce="0,0,0") {
    angularDampingOverride=0.999
    angularSpringConstant=300
    bAngularSpring=true
    bChain=true
    bOverrideAngularDamping=true
    bUseGravityOverride=true
    simulationSpace=RootRelative
    alphaBoolBlend {blendOption=Linear}
    boundBone {boneName=dyn_jacket_side_01_r}
    chainEnd {boneName=dyn_jacket_side_02_r}
    componentAppliedLinearAccClamp=<3000 3000 3000>
    componentLinearAccScale=<0.3 0.3 0.3>
    componentLinearVelScale=<0.3 0.3 0.3>
    gravityOverride {z=-980}
    physicsBodyDefinitions [
     {
      boundBone {boneName=dyn_jacket_side_01_r}
      boxExtents=<20 10 20>
      constraintSetup {angularLimitsMax=<2 0 2>, angularLimitsMin=<-2 -25 -2>, angularTarget=<0 0 0>}
      localJointOffset=<3 0 0>
     }
     {
      boundBone {boneName=dyn_jacket_side_02_r}
      boxExtents=<20 20 20>
      constraintSetup {angularLimitsMax=<2 5 2>, angularLimitsMin=<-2 -15 -2>, angularTarget=<0 0 0>}
      localJointOffset=<3 0 0>
     }
    ]
    planarLimits [{drivingBone {boneName=spine_01}, planeTransform {location=<0 0 12>, rotation=<rot 0 0 0>}}]
    relativeSpaceBone {boneName=root}
    sphericalLimits [
     {limitRadius=10, drivingBone {boneName=spine_02}, sphereLocalOffset=<0 0 -2>}
     {limitRadius=10, drivingBone {boneName=spine_04}, sphereLocalOffset=<0 0 -2>}
     {limitRadius=10, drivingBone {boneName=spine_03}, sphereLocalOffset=<0 0 -2>}
     {limitRadius=10, drivingBone {boneName=spine_05}, sphereLocalOffset=<0 0 -2>}
     {limitRadius=10, drivingBone {boneName=lowerarm_r}, sphereLocalOffset=<0 0 0>}
    ]
   }
   n17 = Anim Dynamics - Bone: dyn_jacket_front_01_r(ComponentPose=@n16, Alpha=1, ExternalForce="0,0,0") {
    angularDampingOverride=0.7
    angularSpringConstant=500
    bAngularSpring=true
    bOverrideAngularDamping=true
    bUseGravityOverride=true
    bUsePlanarLimit=false
    bUseSphericalLimits=true
    simulationSpace=RootRelative
    alphaBoolBlend {blendOption=Linear}
    boundBone {boneName=dyn_jacket_front_01_r}
    componentAppliedLinearAccClamp=<3000 3000 3000>
    componentLinearAccScale=<0.3 0.3 0.3>
    componentLinearVelScale=<0.3 0.3 0.3>
    gravityOverride {z=-980}
    physicsBodyDefinitions [{
      boundBone {boneName=dyn_jacket_front_01_r}
      boxExtents=<30 30 30>
      constraintSetup {angularLimitsMax=<10 0 5>, angularLimitsMin=<-5 -30 -5>, angularTarget=<0 0 0>}
      localJointOffset=<1 0 0>
     }]
    relativeSpaceBone {boneName=root}
    sphericalLimits [
     {limitRadius=10, drivingBone {boneName=spine_02}, sphereLocalOffset=<0 0 0>}
     {limitRadius=10, drivingBone {boneName=spine_04}, sphereLocalOffset=<0 0 0>}
     {limitRadius=10, drivingBone {boneName=spine_03}, sphereLocalOffset=<0 0 0>}
     {limitRadius=10, drivingBone {boneName=spine_05}, sphereLocalOffset=<0 0 0>}
     {limitRadius=2, drivingBone {boneName=hand_r}, sphereLocalOffset=<0 0 0>}
    ]
   }
   n18 = Anim Dynamics - Bone: dyn_jacket_front_02_r(ComponentPose=@n17, Alpha=1, ExternalForce="0,0,0") {
    angularDampingOverride=0.7
    angularSpringConstant=500
    bAngularSpring=true
    bOverrideAngularDamping=true
    bUseGravityOverride=true
    bUsePlanarLimit=false
    bUseSphericalLimits=true
    simulationSpace=RootRelative
    alphaBoolBlend {blendOption=Linear}
    boundBone {boneName=dyn_jacket_front_02_r}
    componentAppliedLinearAccClamp=<3000 3000 3000>
    componentLinearAccScale=<0.3 0.3 0.3>
    componentLinearVelScale=<0.3 0.3 0.3>
    gravityOverride {z=-980}
    physicsBodyDefinitions [{
      boundBone {boneName=dyn_jacket_front_02_r}
      boxExtents=<30 30 30>
      constraintSetup {angularLimitsMax=<10 5 10>, angularLimitsMin=<-5 -20 -5>, angularTarget=<0 0 0>}
      localJointOffset=<1 0 0>
     }]
    relativeSpaceBone {boneName=root}
    sphericalLimits [
     {limitRadius=10, drivingBone {boneName=spine_02}, sphereLocalOffset=<0 0 0>}
     {limitRadius=10, drivingBone {boneName=spine_04}, sphereLocalOffset=<0 0 0>}
     {limitRadius=10, drivingBone {boneName=spine_03}, sphereLocalOffset=<0 0 0>}
     {limitRadius=10, drivingBone {boneName=spine_05}, sphereLocalOffset=<0 0 0>}
     {limitRadius=2, drivingBone {boneName=hand_r}, sphereLocalOffset=<0 0 0>}
    ]
   }
   n19 = Anim Dynamics - Bone: dyn_jacket_front_03_r(ComponentPose=@n18, Alpha=1, ExternalForce="0,0,0") {
    angularDampingOverride=0.7
    angularSpringConstant=500
    bAngularSpring=true
    bOverrideAngularDamping=true
    bUseGravityOverride=true
    bUsePlanarLimit=false
    bUseSphericalLimits=true
    simulationSpace=RootRelative
    alphaBoolBlend {blendOption=Linear}
    boundBone {boneName=dyn_jacket_front_03_r}
    componentAppliedLinearAccClamp=<3000 3000 3000>
    componentLinearAccScale=<0.3 0.3 0.3>
    componentLinearVelScale=<0.3 0.3 0.3>
    gravityOverride {z=-980}
    physicsBodyDefinitions [{
      boundBone {boneName=dyn_jacket_front_03_r}
      boxExtents=<30 30 30>
      constraintSetup {angularLimitsMax=<10 5 10>, angularLimitsMin=<-5 -30 -5>, angularTarget=<0 0 0>}
      localJointOffset=<1 0 0>
     }]
    relativeSpaceBone {boneName=root}
    sphericalLimits [
     {limitRadius=10, drivingBone {boneName=spine_02}, sphereLocalOffset=<0 0 0>}
     {limitRadius=10, drivingBone {boneName=spine_04}, sphereLocalOffset=<0 0 0>}
     {limitRadius=10, drivingBone {boneName=spine_03}, sphereLocalOffset=<0 0 0>}
     {limitRadius=10, drivingBone {boneName=spine_05}, sphereLocalOffset=<0 0 0>}
     {limitRadius=2, drivingBone {boneName=hand_r}, sphereLocalOffset=<0 0 0>}
    ]
   }
   n20 = Anim Dynamics - Bone: dyn_jacket_front_01_l(ComponentPose=@n19, Alpha=1, ExternalForce="0,0,0") {
    angularDampingOverride=0.7
    angularSpringConstant=500
    bAngularSpring=true
    bOverrideAngularDamping=true
    bUseGravityOverride=true
    bUsePlanarLimit=false
    bUseSphericalLimits=true
    simulationSpace=RootRelative
    alphaBoolBlend {blendOption=Linear}
    boundBone {boneName=dyn_jacket_front_01_l}
    componentAppliedLinearAccClamp=<3000 3000 3000>
    componentLinearAccScale=<0.3 0.3 0.3>
    componentLinearVelScale=<0.3 0.3 0.3>
    gravityOverride {z=-980}
    physicsBodyDefinitions [{
      boundBone {boneName=dyn_jacket_front_01_l}
      boxExtents=<30 30 30>
      constraintSetup {angularLimitsMax=<10 0 5>, angularLimitsMin=<-5 -30 -5>, angularTarget=<0 0 0>}
      localJointOffset=<1 0 0>
     }]
    relativeSpaceBone {boneName=root}
    sphericalLimits [
     {limitRadius=10, drivingBone {boneName=spine_02}, sphereLocalOffset=<0 0 0>}
     {limitRadius=10, drivingBone {boneName=spine_04}, sphereLocalOffset=<0 0 0>}
     {limitRadius=10, drivingBone {boneName=spine_03}, sphereLocalOffset=<0 0 0>}
     {limitRadius=10, drivingBone {boneName=spine_05}, sphereLocalOffset=<0 0 0>}
     {limitRadius=2, drivingBone {boneName=hand_r}, sphereLocalOffset=<0 0 0>}
    ]
   }
   n21 = Anim Dynamics - Bone: dyn_jacket_front_02_l(ComponentPose=@n20, Alpha=1, ExternalForce="0,0,0") {
    angularDampingOverride=0.7
    angularSpringConstant=500
    bAngularSpring=true
    bOverrideAngularDamping=true
    bUseGravityOverride=true
    bUsePlanarLimit=false
    bUseSphericalLimits=true
    simulationSpace=RootRelative
    alphaBoolBlend {blendOption=Linear}
    boundBone {boneName=dyn_jacket_front_02_l}
    componentAppliedLinearAccClamp=<3000 3000 3000>
    componentLinearAccScale=<0.3 0.3 0.3>
    componentLinearVelScale=<0.3 0.3 0.3>
    gravityOverride {z=-980}
    physicsBodyDefinitions [{
      boundBone {boneName=dyn_jacket_front_02_l}
      boxExtents=<30 30 30>
      constraintSetup {angularLimitsMax=<10 5 10>, angularLimitsMin=<-5 -20 -5>, angularTarget=<0 0 0>}
      localJointOffset=<1 0 0>
     }]
    relativeSpaceBone {boneName=root}
    sphericalLimits [
     {limitRadius=10, drivingBone {boneName=spine_02}, sphereLocalOffset=<0 0 0>}
     {limitRadius=10, drivingBone {boneName=spine_04}, sphereLocalOffset=<0 0 0>}
     {limitRadius=10, drivingBone {boneName=spine_03}, sphereLocalOffset=<0 0 0>}
     {limitRadius=10, drivingBone {boneName=spine_05}, sphereLocalOffset=<0 0 0>}
     {limitRadius=2, drivingBone {boneName=hand_r}, sphereLocalOffset=<0 0 0>}
    ]
   }
   n22 = Anim Dynamics - Bone: dyn_jacket_front_03_l(ComponentPose=@n21, Alpha=1, ExternalForce="0,0,0") {
    angularDampingOverride=0.7
    angularSpringConstant=500
    bAngularSpring=true
    bOverrideAngularDamping=true
    bUseGravityOverride=true
    bUsePlanarLimit=false
    bUseSphericalLimits=true
    simulationSpace=RootRelative
    alphaBoolBlend {blendOption=Linear}
    boundBone {boneName=dyn_jacket_front_03_l}
    componentAppliedLinearAccClamp=<3000 3000 3000>
    componentLinearAccScale=<0.3 0.3 0.3>
    componentLinearVelScale=<0.3 0.3 0.3>
    gravityOverride {z=-980}
    physicsBodyDefinitions [{
      boundBone {boneName=dyn_jacket_front_03_l}
      boxExtents=<30 30 30>
      constraintSetup {angularLimitsMax=<10 5 10>, angularLimitsMin=<-5 -30 -5>, angularTarget=<0 0 0>}
      localJointOffset=<1 0 0>
     }]
    relativeSpaceBone {boneName=root}
    sphericalLimits [
     {limitRadius=10, drivingBone {boneName=spine_02}, sphereLocalOffset=<0 0 0>}
     {limitRadius=10, drivingBone {boneName=spine_04}, sphereLocalOffset=<0 0 0>}
     {limitRadius=10, drivingBone {boneName=spine_03}, sphereLocalOffset=<0 0 0>}
     {limitRadius=10, drivingBone {boneName=spine_05}, sphereLocalOffset=<0 0 0>}
     {limitRadius=2, drivingBone {boneName=hand_r}, sphereLocalOffset=<0 0 0>}
    ]
   }
   n23 = "Anim Dynamics - Chain: dyn_jacket_center_01 -> dyn_jacket_center_02"(ComponentPose=Local To Component(LocalPose=Use cached pose 'BeforeJacketPose'()), Alpha=0.5, ExternalForce="0,0,0") {
    angularDampingOverride=0.999
    angularSpringConstant=300
    bAngularSpring=true
    bChain=true
    bOverrideAngularDamping=true
    bUseGravityOverride=true
    bUsePlanarLimit=false
    bUseSphericalLimits=true
    linearDampingOverride=0.8
    simulationSpace=RootRelative
    alphaBoolBlend {blendOption=Linear}
    boundBone {boneName=dyn_jacket_center_01}
    chainEnd {boneName=dyn_jacket_center_02}
    componentAppliedLinearAccClamp=<3000 3000 3000>
    componentLinearAccScale=<0.3 0.3 0.3>
    componentLinearVelScale=<0.3 0.3 0.3>
    gravityOverride {z=-980}
    physicsBodyDefinitions [
     {
      boundBone {boneName=dyn_jacket_center_01}
      boxExtents=<20 20 20>
      constraintSetup {angularLimitsMax=<0 -5 1>, angularLimitsMin=<0 -35 -1>, angularTarget=<0 0 0>}
      localJointOffset=<2 0 0>
     }
     {
      boundBone {boneName=dyn_jacket_center_02}
      boxExtents=<20 20 20>
      constraintSetup {angularLimitsMax=<0 0 0>, angularLimitsMin=<0 -15 0>, angularTarget=<0 0 0>}
      localJointOffset=<2 0 0>
     }
    ]
    relativeSpaceBone {boneName=root}
    sphericalLimits [
     {limitRadius=12, drivingBone {boneName=spine_02}, sphereLocalOffset=<0 2 0>}
     {limitRadius=12, drivingBone {boneName=spine_04}, sphereLocalOffset=<0 3 0>}
     {limitRadius=12, drivingBone {boneName=spine_03}, sphereLocalOffset=<0 3 0>}
     {limitRadius=12, drivingBone {boneName=spine_05}, sphereLocalOffset=<0 3 0>}
     {limitRadius=20, drivingBone {boneName=pelvis}, sphereLocalOffset=<0 0 0>}
    ]
   }
   n24 = "Anim Dynamics - Chain: dyn_jacket_back_01_r -> dyn_jacket_back_02_r"(ComponentPose=@n23, Alpha=1, ExternalForce="0,0,0") {
    angularDampingOverride=0.999
    angularSpringConstant=300
    bAngularSpring=true
    bChain=true
    bOverrideAngularDamping=true
    bUseGravityOverride=true
    bUseSphericalLimits=true
    linearDampingOverride=0.8
    simulationSpace=RootRelative
    alphaBoolBlend {blendOption=Linear}
    boundBone {boneName=dyn_jacket_back_01_r}
    chainEnd {boneName=dyn_jacket_back_02_r}
    componentAppliedLinearAccClamp=<3000 3000 3000>
    componentLinearAccScale=<0.3 0.3 0.3>
    componentLinearVelScale=<0.3 0.3 0.3>
    gravityOverride {z=-980}
    physicsBodyDefinitions [
     {
      boundBone {boneName=dyn_jacket_back_01_r}
      boxExtents=<20 20 20>
      constraintSetup {angularLimitsMax=<0 -5 1>, angularLimitsMin=<0 -25 -1>, angularTarget=<0 0 0>}
      localJointOffset=<2 0 0>
     }
     {
      boundBone {boneName=dyn_jacket_back_02_r}
      boxExtents=<20 20 20>
      constraintSetup {angularLimitsMax=<0 0 0>, angularLimitsMin=<0 -15 0>, angularTarget=<0 0 0>}
      localJointOffset=<2 0 0>
     }
    ]
    relativeSpaceBone {boneName=root}
    sphericalLimits [
     {limitRadius=12, drivingBone {boneName=spine_02}, sphereLocalOffset=<0 0 1>}
     {limitRadius=10, drivingBone {boneName=spine_04}, sphereLocalOffset=<0 0 1>}
     {limitRadius=10, drivingBone {boneName=spine_03}, sphereLocalOffset=<0 0 1>}
     {limitRadius=10, drivingBone {boneName=spine_05}, sphereLocalOffset=<0 0 1>}
     {limitRadius=20, drivingBone {boneName=pelvis}, sphereLocalOffset=<0 0 0>}
    ]
   }
   n25 = "Anim Dynamics - Chain: dyn_jacket_back_01_l -> dyn_jacket_back_02_l"(ComponentPose=@n24, Alpha=1, ExternalForce="0,0,0") {
    angularDampingOverride=0.999
    angularSpringConstant=300
    bAngularSpring=true
    bChain=true
    bOverrideAngularDamping=true
    bUseGravityOverride=true
    bUseSphericalLimits=true
    linearDampingOverride=0.8
    simulationSpace=RootRelative
    alphaBoolBlend {blendOption=Linear}
    boundBone {boneName=dyn_jacket_back_01_l}
    chainEnd {boneName=dyn_jacket_back_02_l}
    componentAppliedLinearAccClamp=<3000 3000 3000>
    componentLinearAccScale=<0.3 0.3 0.3>
    componentLinearVelScale=<0.3 0.3 0.3>
    gravityOverride {z=-980}
    physicsBodyDefinitions [
     {
      boundBone {boneName=dyn_jacket_back_01_l}
      boxExtents=<20 20 20>
      constraintSetup {angularLimitsMax=<0 -5 1>, angularLimitsMin=<0 -25 -1>, angularTarget=<0 0 0>}
      localJointOffset=<3 0 0>
     }
     {
      boundBone {boneName=dyn_jacket_back_02_l}
      boxExtents=<20 20 20>
      constraintSetup {angularLimitsMax=<0 0 0>, angularLimitsMin=<0 -15 0>, angularTarget=<0 0 0>}
      localJointOffset=<3 0 0>
     }
    ]
    relativeSpaceBone {boneName=root}
    sphericalLimits [
     {limitRadius=10, drivingBone {boneName=spine_02}, sphereLocalOffset=<0 0 -1>}
     {limitRadius=10, drivingBone {boneName=spine_04}, sphereLocalOffset=<0 0 -1>}
     {limitRadius=10, drivingBone {boneName=spine_03}, sphereLocalOffset=<0 0 -1>}
     {limitRadius=10, drivingBone {boneName=spine_05}, sphereLocalOffset=<0 0 -1>}
     {limitRadius=20, drivingBone {boneName=pelvis}, sphereLocalOffset=<0 0 0>}
    ]
   }
   n26 = "Anim Dynamics - Chain: dyn_jacket_side_01_l -> dyn_jacket_side_02_l"(ComponentPose=@n25, Alpha=1, ExternalForce="0,0,0") {
    angularDampingOverride=0.999
    angularSpringConstant=300
    bAngularSpring=true
    bChain=true
    bOverrideAngularDamping=true
    bUseGravityOverride=true
    bUseSphericalLimits=true
    simulationSpace=RootRelative
    alphaBoolBlend {blendOption=Linear}
    boundBone {boneName=dyn_jacket_side_01_l}
    chainEnd {boneName=dyn_jacket_side_02_l}
    componentAppliedLinearAccClamp=<3000 3000 3000>
    componentLinearAccScale=<0.3 0.3 0.3>
    componentLinearVelScale=<0.3 0.3 0.3>
    gravityOverride {z=-980}
    physicsBodyDefinitions [
     {
      boundBone {boneName=dyn_jacket_side_01_l}
      boxExtents=<20 10 20>
      constraintSetup {angularLimitsMax=<0 -5 1>, angularLimitsMin=<0 -25 -1>, angularTarget=<0 0 0>}
      localJointOffset=<3 0 0>
     }
     {
      boundBone {boneName=dyn_jacket_side_02_l}
      boxExtents=<20 20 20>
      constraintSetup {angularLimitsMax=<0 0 0>, angularLimitsMin=<0 -15 0>, angularTarget=<0 0 0>}
      localJointOffset=<3 0 0>
     }
    ]
    relativeSpaceBone {boneName=root}
    sphericalLimits [
     {limitRadius=10, drivingBone {boneName=spine_02}, sphereLocalOffset=<0 0 -2>}
     {limitRadius=10, drivingBone {boneName=spine_04}, sphereLocalOffset=<0 0 -2>}
     {limitRadius=10, drivingBone {boneName=spine_03}, sphereLocalOffset=<0 0 -2>}
     {limitRadius=10, drivingBone {boneName=spine_05}, sphereLocalOffset=<0 0 -2>}
    ]
   }
   n27 = "Anim Dynamics - Chain: dyn_jacket_side_01_r -> dyn_jacket_side_02_r"(ComponentPose=@n26, Alpha=1, ExternalForce="0,0,0") {
    angularDampingOverride=0.999
    angularSpringConstant=300
    bAngularSpring=true
    bChain=true
    bOverrideAngularDamping=true
    bUseGravityOverride=true
    bUseSphericalLimits=true
    simulationSpace=RootRelative
    alphaBoolBlend {blendOption=Linear}
    boundBone {boneName=dyn_jacket_side_01_r}
    chainEnd {boneName=dyn_jacket_side_02_r}
    componentAppliedLinearAccClamp=<3000 3000 3000>
    componentLinearAccScale=<0.3 0.3 0.3>
    componentLinearVelScale=<0.3 0.3 0.3>
    gravityOverride {z=-980}
    physicsBodyDefinitions [
     {
      boundBone {boneName=dyn_jacket_side_01_r}
      boxExtents=<20 10 20>
      constraintSetup {angularLimitsMax=<0 -5 1>, angularLimitsMin=<0 -25 -1>, angularTarget=<0 0 0>}
      localJointOffset=<3 0 0>
     }
     {
      boundBone {boneName=dyn_jacket_side_02_r}
      boxExtents=<20 20 20>
      constraintSetup {angularLimitsMax=<0 0 0>, angularLimitsMin=<0 -15 0>, angularTarget=<0 0 0>}
      localJointOffset=<3 0 0>
     }
    ]
    relativeSpaceBone {boneName=root}
    sphericalLimits [
     {limitRadius=10, drivingBone {boneName=spine_02}, sphereLocalOffset=<0 0 -2>}
     {limitRadius=10, drivingBone {boneName=spine_04}, sphereLocalOffset=<0 0 -2>}
     {limitRadius=10, drivingBone {boneName=spine_03}, sphereLocalOffset=<0 0 -2>}
     {limitRadius=10, drivingBone {boneName=spine_05}, sphereLocalOffset=<0 0 -2>}
     {limitRadius=10, drivingBone {boneName=lowerarm_r}, sphereLocalOffset=<0 0 0>}
    ]
   }
   n28 = Anim Dynamics - Bone: dyn_jacket_front_01_r(ComponentPose=@n27, Alpha=1, ExternalForce="0,0,0") {
    angularDampingOverride=0.7
    angularSpringConstant=500
    bAngularSpring=true
    bOverrideAngularDamping=true
    bUseGravityOverride=true
    bUseSphericalLimits=true
    simulationSpace=RootRelative
    alphaBoolBlend {blendOption=Linear}
    boundBone {boneName=dyn_jacket_front_01_r}
    componentAppliedLinearAccClamp=<3000 3000 3000>
    componentLinearAccScale=<0.3 0.3 0.3>
    componentLinearVelScale=<0.3 0.3 0.3>
    gravityOverride {z=-980}
    physicsBodyDefinitions [{
      boundBone {boneName=dyn_jacket_front_01_r}
      boxExtents=<30 30 30>
      constraintSetup {angularLimitsMax=<10 0 5>, angularLimitsMin=<-5 -30 -5>, angularTarget=<0 0 0>}
      localJointOffset=<1 0 0>
     }]
    relativeSpaceBone {boneName=root}
    sphericalLimits [
     {limitRadius=10, drivingBone {boneName=spine_02}, sphereLocalOffset=<0 0 0>}
     {limitRadius=10, drivingBone {boneName=spine_04}, sphereLocalOffset=<0 0 0>}
     {limitRadius=10, drivingBone {boneName=spine_03}, sphereLocalOffset=<0 0 0>}
     {limitRadius=10, drivingBone {boneName=spine_05}, sphereLocalOffset=<0 0 0>}
     {limitRadius=7, drivingBone {boneName=lowerarm_r}, sphereLocalOffset=<0 0 0>}
    ]
   }
   n29 = Anim Dynamics - Bone: dyn_jacket_front_02_r(ComponentPose=@n28, Alpha=1, ExternalForce="0,0,0") {
    angularDampingOverride=0.7
    angularSpringConstant=500
    bAngularSpring=true
    bOverrideAngularDamping=true
    bUseGravityOverride=true
    bUseSphericalLimits=true
    simulationSpace=RootRelative
    alphaBoolBlend {blendOption=Linear}
    boundBone {boneName=dyn_jacket_front_02_r}
    componentAppliedLinearAccClamp=<3000 3000 3000>
    componentLinearAccScale=<0.3 0.3 0.3>
    componentLinearVelScale=<0.3 0.3 0.3>
    gravityOverride {z=-980}
    physicsBodyDefinitions [{
      boundBone {boneName=dyn_jacket_front_02_r}
      boxExtents=<30 30 30>
      constraintSetup {angularLimitsMax=<10 5 10>, angularLimitsMin=<-5 -20 -5>, angularTarget=<0 0 0>}
      localJointOffset=<1 0 0>
     }]
    relativeSpaceBone {boneName=root}
    sphericalLimits [
     {limitRadius=10, drivingBone {boneName=spine_02}, sphereLocalOffset=<0 0 0>}
     {limitRadius=10, drivingBone {boneName=spine_04}, sphereLocalOffset=<0 0 0>}
     {limitRadius=10, drivingBone {boneName=spine_03}, sphereLocalOffset=<0 0 0>}
     {limitRadius=10, drivingBone {boneName=spine_05}, sphereLocalOffset=<0 0 0>}
     {limitRadius=7, drivingBone {boneName=lowerarm_r}, sphereLocalOffset=<0 0 0>}
    ]
   }
   n30 = Anim Dynamics - Bone: dyn_jacket_front_03_r(ComponentPose=@n29, Alpha=1, ExternalForce="0,0,0") {
    angularDampingOverride=0.7
    angularSpringConstant=500
    bAngularSpring=true
    bOverrideAngularDamping=true
    bUseGravityOverride=true
    bUseSphericalLimits=true
    simulationSpace=RootRelative
    alphaBoolBlend {blendOption=Linear}
    boundBone {boneName=dyn_jacket_front_03_r}
    componentAppliedLinearAccClamp=<3000 3000 3000>
    componentLinearAccScale=<0.3 0.3 0.3>
    componentLinearVelScale=<0.3 0.3 0.3>
    gravityOverride {z=-980}
    physicsBodyDefinitions [{
      boundBone {boneName=dyn_jacket_front_03_r}
      boxExtents=<30 30 30>
      constraintSetup {angularLimitsMax=<10 5 10>, angularLimitsMin=<-5 -30 -5>, angularTarget=<0 0 0>}
      localJointOffset=<1 0 0>
     }]
    relativeSpaceBone {boneName=root}
    sphericalLimits [
     {limitRadius=10, drivingBone {boneName=spine_02}, sphereLocalOffset=<0 0 0>}
     {limitRadius=10, drivingBone {boneName=spine_04}, sphereLocalOffset=<0 0 0>}
     {limitRadius=10, drivingBone {boneName=spine_03}, sphereLocalOffset=<0 0 0>}
     {limitRadius=10, drivingBone {boneName=spine_05}, sphereLocalOffset=<0 0 0>}
     {limitRadius=7, drivingBone {boneName=lowerarm_r}, sphereLocalOffset=<0 0 0>}
    ]
   }
   n31 = Anim Dynamics - Bone: dyn_jacket_front_01_l(ComponentPose=@n30, Alpha=1, ExternalForce="0,0,0") {
    angularDampingOverride=0.7
    angularSpringConstant=500
    bAngularSpring=true
    bOverrideAngularDamping=true
    bUseGravityOverride=true
    bUseSphericalLimits=true
    simulationSpace=RootRelative
    alphaBoolBlend {blendOption=Linear}
    boundBone {boneName=dyn_jacket_front_01_l}
    componentAppliedLinearAccClamp=<3000 3000 3000>
    componentLinearAccScale=<0.3 0.3 0.3>
    componentLinearVelScale=<0.3 0.3 0.3>
    gravityOverride {z=-980}
    physicsBodyDefinitions [{
      boundBone {boneName=dyn_jacket_front_01_l}
      boxExtents=<30 30 30>
      constraintSetup {angularLimitsMax=<10 0 5>, angularLimitsMin=<-5 -30 -5>, angularTarget=<0 0 0>}
      localJointOffset=<1 0 0>
     }]
    relativeSpaceBone {boneName=root}
    sphericalLimits [
     {limitRadius=10, drivingBone {boneName=spine_02}, sphereLocalOffset=<0 0 0>}
     {limitRadius=10, drivingBone {boneName=spine_04}, sphereLocalOffset=<0 0 0>}
     {limitRadius=10, drivingBone {boneName=spine_03}, sphereLocalOffset=<0 0 0>}
     {limitRadius=10, drivingBone {boneName=spine_05}, sphereLocalOffset=<0 0 0>}
     {limitRadius=2, drivingBone {boneName=hand_r}, sphereLocalOffset=<0 0 0>}
    ]
   }
   n32 = Anim Dynamics - Bone: dyn_jacket_front_02_l(ComponentPose=@n31, Alpha=1, ExternalForce="0,0,0") {
    angularDampingOverride=0.7
    angularSpringConstant=500
    bAngularSpring=true
    bOverrideAngularDamping=true
    bUseGravityOverride=true
    bUseSphericalLimits=true
    simulationSpace=RootRelative
    alphaBoolBlend {blendOption=Linear}
    boundBone {boneName=dyn_jacket_front_02_l}
    componentAppliedLinearAccClamp=<3000 3000 3000>
    componentLinearAccScale=<0.3 0.3 0.3>
    componentLinearVelScale=<0.3 0.3 0.3>
    gravityOverride {z=-980}
    physicsBodyDefinitions [{
      boundBone {boneName=dyn_jacket_front_02_l}
      boxExtents=<30 30 30>
      constraintSetup {angularLimitsMax=<10 5 10>, angularLimitsMin=<-5 -20 -5>, angularTarget=<0 0 0>}
      localJointOffset=<1 0 0>
     }]
    relativeSpaceBone {boneName=root}
    sphericalLimits [
     {limitRadius=10, drivingBone {boneName=spine_02}, sphereLocalOffset=<0 0 0>}
     {limitRadius=10, drivingBone {boneName=spine_04}, sphereLocalOffset=<0 0 0>}
     {limitRadius=10, drivingBone {boneName=spine_03}, sphereLocalOffset=<0 0 0>}
     {limitRadius=10, drivingBone {boneName=spine_05}, sphereLocalOffset=<0 0 0>}
     {limitRadius=2, drivingBone {boneName=hand_r}, sphereLocalOffset=<0 0 0>}
    ]
   }
   n33 = Anim Dynamics - Bone: dyn_jacket_front_03_l(ComponentPose=@n32, Alpha=1, ExternalForce="0,0,0") {
    angularDampingOverride=0.7
    angularSpringConstant=500
    bAngularSpring=true
    bOverrideAngularDamping=true
    bUseGravityOverride=true
    bUseSphericalLimits=true
    simulationSpace=RootRelative
    alphaBoolBlend {blendOption=Linear}
    boundBone {boneName=dyn_jacket_front_03_l}
    componentAppliedLinearAccClamp=<3000 3000 3000>
    componentLinearAccScale=<0.3 0.3 0.3>
    componentLinearVelScale=<0.3 0.3 0.3>
    gravityOverride {z=-980}
    physicsBodyDefinitions [{
      boundBone {boneName=dyn_jacket_front_03_l}
      boxExtents=<30 30 30>
      constraintSetup {angularLimitsMax=<10 5 10>, angularLimitsMin=<-5 -30 -5>, angularTarget=<0 0 0>}
      localJointOffset=<1 0 0>
     }]
    relativeSpaceBone {boneName=root}
    sphericalLimits [
     {limitRadius=10, drivingBone {boneName=spine_02}, sphereLocalOffset=<0 0 0>}
     {limitRadius=10, drivingBone {boneName=spine_04}, sphereLocalOffset=<0 0 0>}
     {limitRadius=10, drivingBone {boneName=spine_03}, sphereLocalOffset=<0 0 0>}
     {limitRadius=10, drivingBone {boneName=spine_05}, sphereLocalOffset=<0 0 0>}
     {limitRadius=2, drivingBone {boneName=hand_r}, sphereLocalOffset=<0 0 0>}
    ]
   }
   n34 = "Transform (Modify) Bone - Bone: mstr_dyn_jacket_center_01"!(ComponentPose=@n33, bAlphaBoolEnabled:bool=Get bIsDBNO(), Translation="0,0,0", Rotation="0,0,-10", Scale="1,1,1") -> (Pose) {
    alphaInputType=Bool
    rotationMode=BMM_Additive
    alphaBoolBlend {blendOption=Linear}
    boneToModify {boneName=mstr_dyn_jacket_center_01}
    rotation {roll=-10}
   }
   n35 = "Anim Dynamics - Chain: dyn_jacket_center_01 -> dyn_jacket_center_02"(ComponentPose=Local To Component(LocalPose=Use cached pose 'BeforeJacketPose'()), Alpha=0.5, ExternalForce="0,0,0") {
    angularDampingOverride=0.999
    angularSpringConstant=300
    bAngularSpring=true
    bChain=true
    bOverrideAngularDamping=true
    bUseGravityOverride=true
    bUsePlanarLimit=false
    bUseSphericalLimits=true
    linearDampingOverride=0.8
    simulationSpace=RootRelative
    alphaBoolBlend {blendOption=Linear}
    boundBone {boneName=dyn_jacket_center_01}
    chainEnd {boneName=dyn_jacket_center_02}
    componentAppliedLinearAccClamp=<3000 3000 3000>
    componentLinearAccScale=<0.3 0.3 0.3>
    componentLinearVelScale=<0.3 0.3 0.3>
    gravityOverride {z=-980}
    physicsBodyDefinitions [
     {
      boundBone {boneName=dyn_jacket_center_01}
      boxExtents=<20 20 20>
      constraintSetup {angularLimitsMax=<0 -5 0>, angularLimitsMin=<0 -6 0>, angularTarget=<0 0 0>}
      localJointOffset=<2 0 0>
     }
     {
      boundBone {boneName=dyn_jacket_center_02}
      boxExtents=<20 20 20>
      constraintSetup {angularLimitsMax=<0 0 0>, angularLimitsMin=<0 -1 0>, angularTarget=<0 0 0>}
      localJointOffset=<2 0 0>
     }
    ]
    relativeSpaceBone {boneName=root}
    sphericalLimits [
     {limitRadius=12, drivingBone {boneName=spine_02}, sphereLocalOffset=<0 2 0>}
     {limitRadius=12, drivingBone {boneName=spine_04}, sphereLocalOffset=<0 3 0>}
     {limitRadius=12, drivingBone {boneName=spine_03}, sphereLocalOffset=<0 3 0>}
     {limitRadius=12, drivingBone {boneName=spine_05}, sphereLocalOffset=<0 3 0>}
     {limitRadius=20, drivingBone {boneName=pelvis}, sphereLocalOffset=<0 0 0>}
    ]
   }
   n36 = "Anim Dynamics - Chain: dyn_jacket_back_01_r -> dyn_jacket_back_02_r"(ComponentPose=@n35, Alpha=1, ExternalForce="0,0,0") {
    angularDampingOverride=0.999
    angularSpringConstant=300
    bAngularSpring=true
    bChain=true
    bOverrideAngularDamping=true
    bUseGravityOverride=true
    bUseSphericalLimits=true
    linearDampingOverride=0.8
    simulationSpace=RootRelative
    alphaBoolBlend {blendOption=Linear}
    boundBone {boneName=dyn_jacket_back_01_r}
    chainEnd {boneName=dyn_jacket_back_02_r}
    componentAppliedLinearAccClamp=<3000 3000 3000>
    componentLinearAccScale=<0.3 0.3 0.3>
    componentLinearVelScale=<0.3 0.3 0.3>
    gravityOverride {z=-980}
    physicsBodyDefinitions [
     {
      boundBone {boneName=dyn_jacket_back_01_r}
      boxExtents=<20 20 20>
      constraintSetup {angularLimitsMax=<0 -5 0>, angularLimitsMin=<0 -6 0>, angularTarget=<0 0 0>}
      localJointOffset=<2 0 0>
     }
     {
      boundBone {boneName=dyn_jacket_back_02_r}
      boxExtents=<20 20 20>
      constraintSetup {angularLimitsMax=<0 0 0>, angularLimitsMin=<0 -1 0>, angularTarget=<0 0 0>}
      localJointOffset=<2 0 0>
     }
    ]
    relativeSpaceBone {boneName=root}
    sphericalLimits [
     {limitRadius=12, drivingBone {boneName=spine_02}, sphereLocalOffset=<0 0 1>}
     {limitRadius=10, drivingBone {boneName=spine_04}, sphereLocalOffset=<0 0 1>}
     {limitRadius=10, drivingBone {boneName=spine_03}, sphereLocalOffset=<0 0 1>}
     {limitRadius=10, drivingBone {boneName=spine_05}, sphereLocalOffset=<0 0 1>}
     {limitRadius=20, drivingBone {boneName=pelvis}, sphereLocalOffset=<0 0 0>}
    ]
   }
   n37 = "Anim Dynamics - Chain: dyn_jacket_back_01_l -> dyn_jacket_back_02_l"(ComponentPose=@n36, Alpha=1, ExternalForce="0,0,0") {
    angularDampingOverride=0.999
    angularSpringConstant=300
    bAngularSpring=true
    bChain=true
    bOverrideAngularDamping=true
    bUseGravityOverride=true
    bUseSphericalLimits=true
    linearDampingOverride=0.8
    simulationSpace=RootRelative
    alphaBoolBlend {blendOption=Linear}
    boundBone {boneName=dyn_jacket_back_01_l}
    chainEnd {boneName=dyn_jacket_back_02_l}
    componentAppliedLinearAccClamp=<3000 3000 3000>
    componentLinearAccScale=<0.3 0.3 0.3>
    componentLinearVelScale=<0.3 0.3 0.3>
    gravityOverride {z=-980}
    physicsBodyDefinitions [
     {
      boundBone {boneName=dyn_jacket_back_01_l}
      boxExtents=<20 20 20>
      constraintSetup {angularLimitsMax=<0 -5 0>, angularLimitsMin=<0 -6 0>, angularTarget=<0 0 0>}
      localJointOffset=<3 0 0>
     }
     {
      boundBone {boneName=dyn_jacket_back_02_l}
      boxExtents=<20 20 20>
      constraintSetup {angularLimitsMax=<0 0 0>, angularLimitsMin=<0 -1 0>, angularTarget=<0 0 0>}
      localJointOffset=<3 0 0>
     }
    ]
    relativeSpaceBone {boneName=root}
    sphericalLimits [
     {limitRadius=10, drivingBone {boneName=spine_02}, sphereLocalOffset=<0 0 -1>}
     {limitRadius=10, drivingBone {boneName=spine_04}, sphereLocalOffset=<0 0 -1>}
     {limitRadius=10, drivingBone {boneName=spine_03}, sphereLocalOffset=<0 0 -1>}
     {limitRadius=10, drivingBone {boneName=spine_05}, sphereLocalOffset=<0 0 -1>}
     {limitRadius=20, drivingBone {boneName=pelvis}, sphereLocalOffset=<0 0 0>}
    ]
   }
   n38 = "Anim Dynamics - Chain: dyn_jacket_side_01_l -> dyn_jacket_side_02_l"(ComponentPose=@n37, Alpha=1, ExternalForce="0,0,0") {
    angularDampingOverride=0.999
    angularSpringConstant=300
    bAngularSpring=true
    bChain=true
    bOverrideAngularDamping=true
    bUseGravityOverride=true
    bUseSphericalLimits=true
    simulationSpace=RootRelative
    alphaBoolBlend {blendOption=Linear}
    boundBone {boneName=dyn_jacket_side_01_l}
    chainEnd {boneName=dyn_jacket_side_02_l}
    componentAppliedLinearAccClamp=<3000 3000 3000>
    componentLinearAccScale=<0.3 0.3 0.3>
    componentLinearVelScale=<0.3 0.3 0.3>
    gravityOverride {z=-980}
    physicsBodyDefinitions [
     {
      boundBone {boneName=dyn_jacket_side_01_l}
      boxExtents=<20 10 20>
      constraintSetup {angularLimitsMax=<0 -5 1>, angularLimitsMin=<0 -25 -1>, angularTarget=<0 0 0>}
      localJointOffset=<3 0 0>
     }
     {
      boundBone {boneName=dyn_jacket_side_02_l}
      boxExtents=<20 20 20>
      constraintSetup {angularLimitsMax=<0 0 0>, angularLimitsMin=<0 -15 0>, angularTarget=<0 0 0>}
      localJointOffset=<3 0 0>
     }
    ]
    relativeSpaceBone {boneName=root}
    sphericalLimits [
     {limitRadius=10, drivingBone {boneName=spine_02}, sphereLocalOffset=<0 0 -2>}
     {limitRadius=10, drivingBone {boneName=spine_04}, sphereLocalOffset=<0 0 -2>}
     {limitRadius=10, drivingBone {boneName=spine_03}, sphereLocalOffset=<0 0 -2>}
     {limitRadius=10, drivingBone {boneName=spine_05}, sphereLocalOffset=<0 0 -2>}
    ]
   }
   n39 = "Anim Dynamics - Chain: dyn_jacket_side_01_r -> dyn_jacket_side_02_r"(ComponentPose=@n38, Alpha=1, ExternalForce="0,0,0") {
    angularDampingOverride=0.999
    angularSpringConstant=300
    bAngularSpring=true
    bChain=true
    bOverrideAngularDamping=true
    bUseGravityOverride=true
    bUseSphericalLimits=true
    simulationSpace=RootRelative
    alphaBoolBlend {blendOption=Linear}
    boundBone {boneName=dyn_jacket_side_01_r}
    chainEnd {boneName=dyn_jacket_side_02_r}
    componentAppliedLinearAccClamp=<3000 3000 3000>
    componentLinearAccScale=<0.3 0.3 0.3>
    componentLinearVelScale=<0.3 0.3 0.3>
    gravityOverride {z=-980}
    physicsBodyDefinitions [
     {
      boundBone {boneName=dyn_jacket_side_01_r}
      boxExtents=<20 10 20>
      constraintSetup {angularLimitsMax=<0 -5 1>, angularLimitsMin=<0 -25 -1>, angularTarget=<0 0 0>}
      localJointOffset=<3 0 0>
     }
     {
      boundBone {boneName=dyn_jacket_side_02_r}
      boxExtents=<20 20 20>
      constraintSetup {angularLimitsMax=<0 0 0>, angularLimitsMin=<0 -15 0>, angularTarget=<0 0 0>}
      localJointOffset=<3 0 0>
     }
    ]
    relativeSpaceBone {boneName=root}
    sphericalLimits [
     {limitRadius=10, drivingBone {boneName=spine_02}, sphereLocalOffset=<0 0 -2>}
     {limitRadius=10, drivingBone {boneName=spine_04}, sphereLocalOffset=<0 0 -2>}
     {limitRadius=10, drivingBone {boneName=spine_03}, sphereLocalOffset=<0 0 -2>}
     {limitRadius=10, drivingBone {boneName=spine_05}, sphereLocalOffset=<0 0 -2>}
     {limitRadius=10, drivingBone {boneName=lowerarm_r}, sphereLocalOffset=<0 0 0>}
    ]
   }
   n40 = Anim Dynamics - Bone: dyn_jacket_front_01_r(ComponentPose=@n39, Alpha=1, ExternalForce="0,0,0") {
    angularDampingOverride=0.7
    angularSpringConstant=500
    bAngularSpring=true
    bOverrideAngularDamping=true
    bUseGravityOverride=true
    bUseSphericalLimits=true
    simulationSpace=RootRelative
    alphaBoolBlend {blendOption=Linear}
    boundBone {boneName=dyn_jacket_front_01_r}
    componentAppliedLinearAccClamp=<3000 3000 3000>
    componentLinearAccScale=<0.3 0.3 0.3>
    componentLinearVelScale=<0.3 0.3 0.3>
    gravityOverride {z=-980}
    physicsBodyDefinitions [{
      boundBone {boneName=dyn_jacket_front_01_r}
      boxExtents=<30 30 30>
      constraintSetup {angularLimitsMax=<10 0 5>, angularLimitsMin=<-5 -30 -5>, angularTarget=<0 0 0>}
      localJointOffset=<1 0 0>
     }]
    relativeSpaceBone {boneName=root}
    sphericalLimits [
     {limitRadius=10, drivingBone {boneName=spine_02}, sphereLocalOffset=<0 0 0>}
     {limitRadius=10, drivingBone {boneName=spine_04}, sphereLocalOffset=<0 0 0>}
     {limitRadius=10, drivingBone {boneName=spine_03}, sphereLocalOffset=<0 0 0>}
     {limitRadius=10, drivingBone {boneName=spine_05}, sphereLocalOffset=<0 0 0>}
     {limitRadius=7, drivingBone {boneName=lowerarm_r}, sphereLocalOffset=<0 0 0>}
    ]
   }
   n41 = Anim Dynamics - Bone: dyn_jacket_front_02_r(ComponentPose=@n40, Alpha=1, ExternalForce="0,0,0") {
    angularDampingOverride=0.7
    angularSpringConstant=500
    bAngularSpring=true
    bOverrideAngularDamping=true
    bUseGravityOverride=true
    bUseSphericalLimits=true
    simulationSpace=RootRelative
    alphaBoolBlend {blendOption=Linear}
    boundBone {boneName=dyn_jacket_front_02_r}
    componentAppliedLinearAccClamp=<3000 3000 3000>
    componentLinearAccScale=<0.3 0.3 0.3>
    componentLinearVelScale=<0.3 0.3 0.3>
    gravityOverride {z=-980}
    physicsBodyDefinitions [{
      boundBone {boneName=dyn_jacket_front_02_r}
      boxExtents=<30 30 30>
      constraintSetup {angularLimitsMax=<10 5 10>, angularLimitsMin=<-5 -20 -5>, angularTarget=<0 0 0>}
      localJointOffset=<1 0 0>
     }]
    relativeSpaceBone {boneName=root}
    sphericalLimits [
     {limitRadius=10, drivingBone {boneName=spine_02}, sphereLocalOffset=<0 0 0>}
     {limitRadius=10, drivingBone {boneName=spine_04}, sphereLocalOffset=<0 0 0>}
     {limitRadius=10, drivingBone {boneName=spine_03}, sphereLocalOffset=<0 0 0>}
     {limitRadius=10, drivingBone {boneName=spine_05}, sphereLocalOffset=<0 0 0>}
     {limitRadius=7, drivingBone {boneName=lowerarm_r}, sphereLocalOffset=<0 0 0>}
    ]
   }
   n42 = Anim Dynamics - Bone: dyn_jacket_front_03_r(ComponentPose=@n41, Alpha=1, ExternalForce="0,0,0") {
    angularDampingOverride=0.7
    angularSpringConstant=500
    bAngularSpring=true
    bOverrideAngularDamping=true
    bUseGravityOverride=true
    bUseSphericalLimits=true
    simulationSpace=RootRelative
    alphaBoolBlend {blendOption=Linear}
    boundBone {boneName=dyn_jacket_front_03_r}
    componentAppliedLinearAccClamp=<3000 3000 3000>
    componentLinearAccScale=<0.3 0.3 0.3>
    componentLinearVelScale=<0.3 0.3 0.3>
    gravityOverride {z=-980}
    physicsBodyDefinitions [{
      boundBone {boneName=dyn_jacket_front_03_r}
      boxExtents=<30 30 30>
      constraintSetup {angularLimitsMax=<10 5 10>, angularLimitsMin=<-5 -30 -5>, angularTarget=<0 0 0>}
      localJointOffset=<1 0 0>
     }]
    relativeSpaceBone {boneName=root}
    sphericalLimits [
     {limitRadius=10, drivingBone {boneName=spine_02}, sphereLocalOffset=<0 0 0>}
     {limitRadius=10, drivingBone {boneName=spine_04}, sphereLocalOffset=<0 0 0>}
     {limitRadius=10, drivingBone {boneName=spine_03}, sphereLocalOffset=<0 0 0>}
     {limitRadius=10, drivingBone {boneName=spine_05}, sphereLocalOffset=<0 0 0>}
     {limitRadius=7, drivingBone {boneName=lowerarm_r}, sphereLocalOffset=<0 0 0>}
    ]
   }
   n43 = Anim Dynamics - Bone: dyn_jacket_front_01_l(ComponentPose=@n42, Alpha=1, ExternalForce="0,0,0") {
    angularDampingOverride=0.7
    angularSpringConstant=500
    bAngularSpring=true
    bOverrideAngularDamping=true
    bUseGravityOverride=true
    bUseSphericalLimits=true
    simulationSpace=RootRelative
    alphaBoolBlend {blendOption=Linear}
    boundBone {boneName=dyn_jacket_front_01_l}
    componentAppliedLinearAccClamp=<3000 3000 3000>
    componentLinearAccScale=<0.3 0.3 0.3>
    componentLinearVelScale=<0.3 0.3 0.3>
    gravityOverride {z=-980}
    physicsBodyDefinitions [{
      boundBone {boneName=dyn_jacket_front_01_l}
      boxExtents=<30 30 30>
      constraintSetup {angularLimitsMax=<10 0 5>, angularLimitsMin=<-5 -30 -5>, angularTarget=<0 0 0>}
      localJointOffset=<1 0 0>
     }]
    relativeSpaceBone {boneName=root}
    sphericalLimits [
     {limitRadius=10, drivingBone {boneName=spine_02}, sphereLocalOffset=<0 0 0>}
     {limitRadius=10, drivingBone {boneName=spine_04}, sphereLocalOffset=<0 0 0>}
     {limitRadius=10, drivingBone {boneName=spine_03}, sphereLocalOffset=<0 0 0>}
     {limitRadius=10, drivingBone {boneName=spine_05}, sphereLocalOffset=<0 0 0>}
     {limitRadius=2, drivingBone {boneName=hand_r}, sphereLocalOffset=<0 0 0>}
    ]
   }
   n44 = Anim Dynamics - Bone: dyn_jacket_front_02_l(ComponentPose=@n43, Alpha=1, ExternalForce="0,0,0") {
    angularDampingOverride=0.7
    angularSpringConstant=500
    bAngularSpring=true
    bOverrideAngularDamping=true
    bUseGravityOverride=true
    bUseSphericalLimits=true
    simulationSpace=RootRelative
    alphaBoolBlend {blendOption=Linear}
    boundBone {boneName=dyn_jacket_front_02_l}
    componentAppliedLinearAccClamp=<3000 3000 3000>
    componentLinearAccScale=<0.3 0.3 0.3>
    componentLinearVelScale=<0.3 0.3 0.3>
    gravityOverride {z=-980}
    physicsBodyDefinitions [{
      boundBone {boneName=dyn_jacket_front_02_l}
      boxExtents=<30 30 30>
      constraintSetup {angularLimitsMax=<10 5 10>, angularLimitsMin=<-5 -20 -5>, angularTarget=<0 0 0>}
      localJointOffset=<1 0 0>
     }]
    relativeSpaceBone {boneName=root}
    sphericalLimits [
     {limitRadius=10, drivingBone {boneName=spine_02}, sphereLocalOffset=<0 0 0>}
     {limitRadius=10, drivingBone {boneName=spine_04}, sphereLocalOffset=<0 0 0>}
     {limitRadius=10, drivingBone {boneName=spine_03}, sphereLocalOffset=<0 0 0>}
     {limitRadius=10, drivingBone {boneName=spine_05}, sphereLocalOffset=<0 0 0>}
     {limitRadius=2, drivingBone {boneName=hand_r}, sphereLocalOffset=<0 0 0>}
    ]
   }
   n45 = Anim Dynamics - Bone: dyn_jacket_front_03_l(ComponentPose=@n44, Alpha=1, ExternalForce="0,0,0") {
    angularDampingOverride=0.7
    angularSpringConstant=500
    bAngularSpring=true
    bOverrideAngularDamping=true
    bUseGravityOverride=true
    bUseSphericalLimits=true
    simulationSpace=RootRelative
    alphaBoolBlend {blendOption=Linear}
    boundBone {boneName=dyn_jacket_front_03_l}
    componentAppliedLinearAccClamp=<3000 3000 3000>
    componentLinearAccScale=<0.3 0.3 0.3>
    componentLinearVelScale=<0.3 0.3 0.3>
    gravityOverride {z=-980}
    physicsBodyDefinitions [{
      boundBone {boneName=dyn_jacket_front_03_l}
      boxExtents=<30 30 30>
      constraintSetup {angularLimitsMax=<10 5 10>, angularLimitsMin=<-5 -30 -5>, angularTarget=<0 0 0>}
      localJointOffset=<1 0 0>
     }]
    relativeSpaceBone {boneName=root}
    sphericalLimits [
     {limitRadius=10, drivingBone {boneName=spine_02}, sphereLocalOffset=<0 0 0>}
     {limitRadius=10, drivingBone {boneName=spine_04}, sphereLocalOffset=<0 0 0>}
     {limitRadius=10, drivingBone {boneName=spine_03}, sphereLocalOffset=<0 0 0>}
     {limitRadius=10, drivingBone {boneName=spine_05}, sphereLocalOffset=<0 0 0>}
     {limitRadius=2, drivingBone {boneName=hand_r}, sphereLocalOffset=<0 0 0>}
    ]
   }
   n46 = Blend Poses by bool(
    BlendPose_0=Blend Poses by bool(
     BlendPose_0=Component To Local(ComponentPose=@n22)
     BlendPose_1=Component To Local(ComponentPose=@n11)
     BlendTime_0=0.5
     BlendTime_1=0.5
     bActiveValue=Get IsBackbling() -> (IsBackbling:bool)
    ) {blendType=Linear}
    BlendPose_1=Blend Poses by bool(
     BlendPose_0=Component To Local(ComponentPose=@n45)
     BlendPose_1=Component To Local(ComponentPose=@n34)
     BlendTime_0=0.5
     BlendTime_1=0.5
     bActiveValue=Get IsBackbling()
    ) {blendType=Linear}
    BlendTime_0=0.5
    BlendTime_1=0.5
    bActiveValue=Get bIsParachuteOpened()
   ) {blendType=Linear}
   Output Pose(Result=Component To Local(ComponentPose=Local To Component(LocalPose=@n46))) {layerGroup=None}
   n47 = Anim Dynamics - Bone: dyn_laces_side01_l(ComponentPose=Local To Component(LocalPose=Use cached pose 'ControlRigPose'()), Alpha=1, ExternalForce="0,0,0") {
    angularDampingOverride=0.7
    angularSpringConstant=500
    bAngularSpring=true
    bOverrideAngularDamping=true
    bUseGravityOverride=true
    gravityScale=-1
    simulationSpace=RootRelative
    alphaBoolBlend {blendOption=Linear}
    boundBone {boneName=dyn_laces_side01_l}
    componentAppliedLinearAccClamp=<3000 3000 3000>
    componentLinearAccScale=<0.3 0.3 0.3>
    componentLinearVelScale=<0.3 0.3 0.3>
    gravityOverride {z=-600}
    physicsBodyDefinitions [{
      boundBone {boneName=dyn_laces_side01_l}
      boxExtents=<12 16 8>
      constraintSetup {angularLimitsMax=<0 10 20>, angularLimitsMin=<-30 -55 -10>, angularTarget=<0 0 0>}
      localJointOffset=<3 0 0>
     }]
   }
   n48 = Anim Dynamics - Bone: dyn_laces_side02_l(ComponentPose=@n47, Alpha=1, ExternalForce="0,0,0") {
    angularDampingOverride=0.75
    angularSpringConstant=500
    bAngularSpring=true
    bOverrideAngularDamping=true
    bUseGravityOverride=true
    gravityScale=-1
    simulationSpace=RootRelative
    alphaBoolBlend {blendOption=Linear}
    boundBone {boneName=dyn_laces_side02_l}
    componentAppliedLinearAccClamp=<3000 3000 3000>
    componentLinearAccScale=<0.3 0.3 0.3>
    componentLinearVelScale=<0.3 0.3 0.3>
    gravityOverride {z=-600}
    physicsBodyDefinitions [{
      boundBone {boneName=dyn_laces_side02_l}
      boxExtents=<12 16 8>
      constraintSetup {angularLimitsMax=<0 10 0>, angularLimitsMin=<-30 -55 -20>, angularTarget=<0 0 0>}
      localJointOffset=<3 0 0>
     }]
   }
   n49 = Anim Dynamics - Bone: dyn_laces_side02_r(ComponentPose=@n48, Alpha=1, ExternalForce="0,0,0") {
    angularDampingOverride=0.7
    angularSpringConstant=500
    bAngularSpring=true
    bOverrideAngularDamping=true
    bUseGravityOverride=true
    gravityScale=-1
    simulationSpace=RootRelative
    alphaBoolBlend {blendOption=Linear}
    boundBone {boneName=dyn_laces_side02_r}
    componentAppliedLinearAccClamp=<3000 3000 3000>
    componentLinearAccScale=<0.3 0.3 0.3>
    componentLinearVelScale=<0.3 0.3 0.3>
    gravityOverride {z=-600}
    physicsBodyDefinitions [{
      boundBone {boneName=dyn_laces_side02_r}
      boxExtents=<12 16 8>
      constraintSetup {angularLimitsMax=<0 10 20>, angularLimitsMin=<-30 -55 0>, angularTarget=<0 0 0>}
      localJointOffset=<3 0 0>
     }]
   }
   n50 = Anim Dynamics - Bone: dyn_laces_side01_r(ComponentPose=@n49, Alpha=1, ExternalForce="0,0,0") {
    angularDampingOverride=0.75
    angularSpringConstant=500
    bAngularSpring=true
    bOverrideAngularDamping=true
    bUseGravityOverride=true
    gravityScale=-1
    simulationSpace=RootRelative
    alphaBoolBlend {blendOption=Linear}
    boundBone {boneName=dyn_laces_side01_r}
    componentAppliedLinearAccClamp=<3000 3000 3000>
    componentLinearAccScale=<0.3 0.3 0.3>
    componentLinearVelScale=<0.3 0.3 0.3>
    gravityOverride {z=-600}
    physicsBodyDefinitions [{
      boundBone {boneName=dyn_laces_side01_r}
      boxExtents=<12 16 8>
      constraintSetup {angularLimitsMax=<0 10 0>, angularLimitsMin=<-30 -55 -20>, angularTarget=<0 0 0>}
      localJointOffset=<3 0 0>
     }]
   }
   n51 = Anim Dynamics - Bone: dyn_fur_top_01_r(ComponentPose=@n50, Alpha=1, ExternalForce="0,0,0") {
    angularDampingOverride=0.7
    angularSpringConstant=500
    bAngularSpring=true
    bOverrideAngularDamping=true
    bUseGravityOverride=true
    gravityScale=-1
    simulationSpace=RootRelative
    alphaBoolBlend {blendOption=Linear}
    boundBone {boneName=dyn_fur_top_01_r}
    componentAppliedLinearAccClamp=<3000 3000 3000>
    componentLinearAccScale=<0.3 0.3 0.3>
    componentLinearVelScale=<0.3 0.3 0.3>
    gravityOverride {z=-600}
    physicsBodyDefinitions [{
      boundBone {boneName=dyn_fur_top_01_r}
      boxExtents=<12 16 8>
      constraintSetup {angularLimitsMax=<0 0 20>, angularLimitsMin=<-30 -50 -10>, angularTarget=<0 0 0>}
      localJointOffset=<3 0 0>
     }]
   }
   n52 = Anim Dynamics - Bone: dyn_fur_back_01_r(ComponentPose=@n51, Alpha=1, ExternalForce="0,0,0") {
    angularDampingOverride=0.7
    angularSpringConstant=500
    bAngularSpring=true
    bOverrideAngularDamping=true
    bUseGravityOverride=true
    gravityScale=-1
    simulationSpace=RootRelative
    alphaBoolBlend {blendOption=Linear}
    boundBone {boneName=dyn_fur_back_01_r}
    componentAppliedLinearAccClamp=<3000 3000 3000>
    componentLinearAccScale=<0.3 0.3 0.3>
    componentLinearVelScale=<0.3 0.3 0.3>
    gravityOverride {z=-600}
    physicsBodyDefinitions [{
      boundBone {boneName=dyn_fur_back_01_r}
      boxExtents=<12 16 8>
      constraintSetup {angularLimitsMax=<0 3 0>, angularLimitsMin=<-30 -4 -30>, angularTarget=<0 0 0>}
      localJointOffset=<3 0 0>
     }]
   }
   n53 = Anim Dynamics - Bone: dyn_fur_front_01_r(ComponentPose=@n52, Alpha=1, ExternalForce="0,0,0") {
    angularDampingOverride=0.7
    angularSpringConstant=500
    bAngularSpring=true
    bOverrideAngularDamping=true
    bUseGravityOverride=true
    gravityScale=-1
    simulationSpace=RootRelative
    alphaBoolBlend {blendOption=Linear}
    boundBone {boneName=dyn_fur_front_01_r}
    componentAppliedLinearAccClamp=<3000 3000 3000>
    componentLinearAccScale=<0.3 0.3 0.3>
    componentLinearVelScale=<0.3 0.3 0.3>
    gravityOverride {z=-600}
    physicsBodyDefinitions [{
      boundBone {boneName=dyn_fur_front_01_r}
      boxExtents=<12 16 8>
      constraintSetup {angularLimitsMax=<0 0 20>, angularLimitsMin=<-30 -55 -10>, angularTarget=<0 0 0>}
      localJointOffset=<3 0 0>
     }]
   }
   n54 = Anim Dynamics - Bone: dyn_fur_top_01_l(ComponentPose=@n53, Alpha=1, ExternalForce="0,0,0") {
    angularDampingOverride=0.7
    angularSpringConstant=500
    bAngularSpring=true
    bOverrideAngularDamping=true
    bUseGravityOverride=true
    gravityScale=-1
    simulationSpace=RootRelative
    alphaBoolBlend {blendOption=Linear}
    boundBone {boneName=dyn_fur_top_01_l}
    componentAppliedLinearAccClamp=<3000 3000 3000>
    componentLinearAccScale=<0.3 0.3 0.3>
    componentLinearVelScale=<0.3 0.3 0.3>
    gravityOverride {z=-600}
    physicsBodyDefinitions [{
      boundBone {boneName=dyn_fur_top_01_l}
      boxExtents=<12 16 8>
      constraintSetup {angularLimitsMax=<0 0 20>, angularLimitsMin=<-30 -50 -10>, angularTarget=<0 0 0>}
      localJointOffset=<3 0 0>
     }]
   }
   n55 = Anim Dynamics - Bone: dyn_fur_back_01_l(ComponentPose=@n54, Alpha=1, ExternalForce="0,0,0") {
    angularDampingOverride=0.7
    angularSpringConstant=500
    bAngularSpring=true
    bOverrideAngularDamping=true
    bUseGravityOverride=true
    gravityScale=-1
    simulationSpace=RootRelative
    alphaBoolBlend {blendOption=Linear}
    boundBone {boneName=dyn_fur_back_01_l}
    componentAppliedLinearAccClamp=<3000 3000 3000>
    componentLinearAccScale=<0.3 0.3 0.3>
    componentLinearVelScale=<0.3 0.3 0.3>
    gravityOverride {z=-600}
    physicsBodyDefinitions [{
      boundBone {boneName=dyn_fur_back_01_l}
      boxExtents=<12 16 8>
      constraintSetup {angularLimitsMax=<0 3 0>, angularLimitsMin=<-30 -4 -30>, angularTarget=<0 0 0>}
      localJointOffset=<3 0 0>
     }]
   }
   n56 = Anim Dynamics - Bone: dyn_fur_top_01_l(ComponentPose=@n55, Alpha=1, ExternalForce="0,0,0") {
    angularDampingOverride=0.7
    angularSpringConstant=500
    bAngularSpring=true
    bOverrideAngularDamping=true
    bUseGravityOverride=true
    gravityScale=-1
    simulationSpace=RootRelative
    alphaBoolBlend {blendOption=Linear}
    boundBone {boneName=dyn_fur_top_01_l}
    componentAppliedLinearAccClamp=<3000 3000 3000>
    componentLinearAccScale=<0.3 0.3 0.3>
    componentLinearVelScale=<0.3 0.3 0.3>
    gravityOverride {z=-600}
    physicsBodyDefinitions [{
      boundBone {boneName=dyn_fur_top_01_l}
      boxExtents=<12 16 8>
      constraintSetup {angularLimitsMax=<0 0 20>, angularLimitsMin=<-30 -55 -10>, angularTarget=<0 0 0>}
      localJointOffset=<3 0 0>
     }]
   }
   Save cached pose 'BeforeJacketPose'(Pose=Component To Local(ComponentPose=@n56)) {cacheName=BeforeJacketPose}
   n57 = Control Rig(Source=Copy Pose From Mesh() {bUseAttachedParent=true}) {
    alphaBoolBlend {blendOption=Linear}
    controlRigAssetReference {blueprintRigClass=&/F_MED_WonderHill_CtrlRig.F_MED_WonderHill_CtrlRig_C}
   }
   n58 = "Transform (Modify) Bone - Bone: sleeve_fr_R"!(
    ComponentPose=Local To Component(LocalPose=@n57)
    Alpha="Get right front sleeve rot Y "() -> ("right front sleeve rot Y ":float)
    Translation="0,0,-5"
    Rotation="0,0,0"
   ) -> (Pose) {
    rotationMode=BMM_Additive
    rotationSpace=BCS_BoneSpace
    translationMode=BMM_Additive
    translationSpace=BCS_BoneSpace
    alphaBoolBlend {blendOption=Linear}
    boneToModify {boneName=sleeve_fr_R}
    translation {z=-5}
   }
   n59 = "Transform (Modify) Bone - Bone: sleeve_fr_R"(
    ComponentPose=@n58
    Alpha=Get right front sleeve rot Z() -> (right front sleeve rot Z:float)
    Translation="-5.547,-0.453,-1.337"
    Rotation="-29.863,0,0"
   ) {
    rotationMode=BMM_Additive
    rotationSpace=BCS_BoneSpace
    translationMode=BMM_Additive
    translationSpace=BCS_BoneSpace
    alphaBoolBlend {blendOption=Linear}
    boneToModify {boneName=sleeve_fr_R}
    rotation {pitch=-29.863}
    translation=<-5.547 -0.453 -1.337>
   }
   n60 = "Transform (Modify) Bone - Bone: sleeve_bk_L"!(ComponentPose=@n59, Alpha=Get back sleeve L() -> (back sleeve L:float), Translation="0,-1,2.5", Rotation="0,0,0") -> (Pose) {
    rotationMode=BMM_Additive
    rotationSpace=BCS_BoneSpace
    translationMode=BMM_Additive
    translationSpace=BCS_BoneSpace
    alphaBoolBlend {blendOption=Linear}
    boneToModify {boneName=sleeve_bk_L}
    translation {y=-1, z=2.5}
   }
   n61 = "Transform (Modify) Bone - Bone: sleeve_bk_R"!(ComponentPose=@n60, Alpha=Get back sleeve R() -> (back sleeve R:float), Translation="0,-1,-2.5", Rotation="0,0,0") -> (Pose) {
    rotationMode=BMM_Additive
    rotationSpace=BCS_BoneSpace
    translationMode=BMM_Additive
    translationSpace=BCS_BoneSpace
    alphaBoolBlend {blendOption=Linear}
    boneToModify {boneName=sleeve_bk_R}
    translation {y=-1, z=-2.5}
   }
   Save cached pose 'ControlRigPose'(Pose=Component To Local(ComponentPose=@n61)) {cacheName=ControlRigPose}
   Get SkydiveFloatingAlpha() -> (SkydiveFloatingAlpha:double)
  ]}
 EventGraph {nodes [
   Event BlueprintUpdateAnimation()
   > Set SkydiveFloatingAlpha(SkydiveFloatingAlpha:double=SelectFloat(A=0.5, B=1, bPickA=Get bIsSkydiveFloating())) -> (Output_Get:double)
   n1 = CheckForBackbling()
   n2 = OR Boolean!(A=@n1, B=@n1.bIsBackblingACape, C:bool=@n1.bIsBackblingLong) -> (ReturnValue)
   > Set IsBackbling(IsBackbling:bool=@n2) -> (Output_Get:bool)
   > Set IsBackblingAlpha(IsBackblingAlpha:double=SelectFloat(A=1, B=0, bPickA=@n2)) -> (Output_Get:double)
   n3 = Get SkelMesh()
   > Set back sleeve L(back sleeve L:float=SelectFloat(
     A=GetAlphaFromReferenceTransformDelta(SkeletalMeshComponent=@n3, BoneName=upperarm_l, BaseName=spine_05, DeltaComponentType=EulerY, InputRange=(X=0,Y=-10), TwistAxis="1,0,0")
     B=0
     bPickA=Get bIsInFrontEnd()
    )) -> (Output_Get:float)
   > Set back sleeve R(back sleeve R:float=SelectFloat(
     A=GetAlphaFromReferenceTransformDelta(SkeletalMeshComponent=@n3, BoneName=upperarm_r, BaseName=spine_05, DeltaComponentType=EulerY, InputRange=(X=0,Y=-10), TwistAxis="1,0,0")
     B=0
     bPickA=Get bIsInFrontEnd()
    )) -> (Output_Get:float)
   > "Set right front sleeve rot Y "("right front sleeve rot Y ":float=SelectFloat(
     A=GetAlphaFromReferenceTransformDelta(SkeletalMeshComponent=@n3, BoneName=upperarm_r, BaseName=spine_05, DeltaComponentType=EulerY, InputRange=(X=0,Y=-30), TwistAxis="1,0,0")
     B=0
     bPickA=Get bIsCrouching()
    )) -> (Output_Get:float)
   > Set right front sleeve rot Z(right front sleeve rot Z:float=SelectFloat(
     A=GetAlphaFromReferenceTransformDelta(SkeletalMeshComponent=@n3, BoneName=upperarm_r, BaseName=spine_05, InputRange=(X=0,Y=150), TwistAxis="1,0,0")
     B=0
     bPickA=Get bIsCrouching()
    )) -> (Output_Get:float)
   Event BlueprintInitializeAnimation()
   > Set SkelMesh(SkelMesh=GetOwnerSkeletalMeshComponent())
  ]}
]
variables [
 back sleeve L {type=number}
 back sleeve R {type=number}
 Is Skydive Floating {type=boolean}
 IsBackbling {type=boolean}
 IsBackblingAlpha {type=number}
 "right front sleeve rot Y " {type=number}
 right front sleeve rot Z {type=number}
 SkelMesh {type=SkeletalMeshComponent}
 SkydiveFloatingAlpha {type=number}
]
