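# Control Rig for the WonderHill female cosmetic's face accessory part, applying pose-driven offsets to the back hair bones.
# Spherical pose readers on the head and right clavicle weight the ModifyTransforms nodes, run through a four-output aggregate sequence.
# The clavicle correction runs only when `not skydiving` is true; a second head reader is scaled by `is melee float`.
# The upper-arm reader with `is targeting float` and the `is melee` branch drive nothing.
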ControlRigBlueprint /BRCosmetics/Characters/Player/Female/Medium/Bodies/F_MED_WonderHill/Meshes/Parts/F_MED_WonderHill_FaceAcc_CtrlRig
graphs [
 RigVMFunctionLibrary {links [], nodes []}
 RigVMFunction_Sequence_1_ContainedGraph {nodes [
   n1 = RigVMFunctionEntryNode()
   RigVMFunction_Sequence = &RigVMFunction_Sequence(ExecuteContext=@n1.ExecuteContext)
   n2 = &RigVMFunction_Sequence(ExecuteContext=@RigVMFunction_Sequence.B)
   n3 = &RigVMFunction_Sequence(ExecuteContext=@n2.B)
   RigVMFunctionReturnNode(A=@RigVMFunction_Sequence.A, B=@n2.A, C=@n3.A, D=@n3.B)
  ]}
 RigVMModel {nodes [
   RigUnit_BeginExecution = &RigUnit_BeginExecution()
   RigVMFunction_Sequence = RigVMAggregateNode(ExecuteContext=@RigUnit_BeginExecution.ExecutePin)
   n1 = &RigUnit_SphericalPoseReader(
    ActiveRegionScaleFactors=(PositiveWidth=0.5,NegativeWidth=1,PositiveHeight=1,NegativeHeight=1)
    ActiveRegionSize=0.1
    Debug=(bDrawDebug=false,bDraw2D=false,bDrawLocalAxes=false,DebugScale=25,DebugSegments=20,DebugThickness=0.25)
    DriverAxis=(X=0,Y=-1,Z=0)
    DriverItem=(Type=Bone,Name="head")
    FalloffRegionScaleFactors=(PositiveWidth=0.7,NegativeWidth=1,PositiveHeight=1,NegativeHeight=1)
    FalloffSize=0.3
    FlipHeightScaling=False
    FlipWidthScaling=False
    OptionalParentItem=(Type=Bone,Name="None")
    RotationOffset=(X=-90,Y=0,Z=-35)
    ExecutePin=@RigVMFunction_Sequence.A
   )
   > &RigUnit_ModifyTransforms(
    ItemToModify=((Item=(Type=Bone,Name="dyn_hair_back_MSTR_MSTR"),Transform=(Rotation=(X=0,Y=0,Z=0,W=1),Translation=(X=-2,Y=-2,Z=0),Scale3D=(X=1,Y=1,Z=1))))
    WeightMinimum=0
    ItemToModify.0.Transform.Rotation=&RigVMFunction_MathQuaternionFromEuler(Euler=(X=0,Y=5,Z=14)).Result
    Weight=@n1.OutputParam
   )
   > &RigUnit_ModifyTransforms(
    ItemToModify="((Item=(Type=Bone,Name=\"dyn_hair_back_middle_MSTR\"),Transform=(Rotation=(X=0,Y=0,Z=0,W=1),Translation=(X=0,Y=0,Z=0),Scale3D=(X=1,Y=1,Z=1))),(Item=(Type=Bone,Name=\"dyn_hair_back_middle_02\"),Transform=(Rotation=(X=0,Y=0,Z=0,W=1),Translation=(X=-1,Y=0,Z=0),Sc ...(truncated)"
    WeightMinimum=0
    ItemToModify.1.Transform.Rotation=&RigVMFunction_MathQuaternionFromEuler(Euler=(X=0,Y=0,Z=10)).Result
    ItemToModify.0.Transform.Rotation=&RigVMFunction_MathQuaternionFromEuler(Euler=(X=0,Y=0,Z=0)).Result
    Weight=@n1.OutputParam
   )
   > &RigUnit_ModifyTransforms(
    ItemToModify="((Item=(Type=Bone,Name=\"dyn_hair_back_right_MSTR\"),Transform=(Rotation=(X=0,Y=0,Z=0,W=1),Translation=(X=-3,Y=0,Z=0),Scale3D=(X=1,Y=1,Z=1))),(Item=(Type=Bone,Name=\"dyn_hair_back_right_02\"),Transform=(Rotation=(X=0,Y=0,Z=0,W=1),Translation=(X=-2,Y=0,Z=0),Sca ...(truncated)"
    WeightMinimum=0
    ItemToModify.1.Transform.Rotation=&RigVMFunction_MathQuaternionFromEuler(Euler=(X=0,Y=0,Z=0)).Result
    ItemToModify.0.Transform.Rotation=&RigVMFunction_MathQuaternionFromEuler(Euler=(X=0,Y=0,Z=5)).Result
    Weight=@n1.OutputParam
   )
   RigVMFunction_ControlFlowBranch = &RigVMFunction_ControlFlowBranch(ExecuteContext=@RigVMFunction_Sequence.B, Condition=RigVMVariableNode() {variable=not skydiving}.Value)
   n2 = &RigUnit_SphericalPoseReader(
    ActiveRegionScaleFactors=(PositiveWidth=1,NegativeWidth=1,PositiveHeight=1,NegativeHeight=1)
    ActiveRegionSize=0.1
    Debug=(bDrawDebug=false,bDraw2D=false,bDrawLocalAxes=false,DebugScale=25,DebugSegments=20,DebugThickness=0.25)
    DriverAxis=(X=-1,Y=0,Z=0)
    DriverItem=(Type=Bone,Name="clavicle_r")
    FalloffRegionScaleFactors=(PositiveWidth=0.7,NegativeWidth=1,PositiveHeight=1,NegativeHeight=1)
    FalloffSize=0.3
    FlipHeightScaling=False
    FlipWidthScaling=False
    OptionalParentItem=(Type=Bone,Name="None")
    RotationOffset=(X=100,Y=0,Z=90)
    ExecutePin=@RigVMFunction_ControlFlowBranch.True
   )
   > &RigUnit_ModifyTransforms(
    ItemToModify="((Item=(Type=Bone,Name=\"dyn_hair_back_middle_MSTR\"),Transform=(Rotation=(X=0,Y=0,Z=0,W=1),Translation=(X=-1,Y=0,Z=0),Scale3D=(X=1,Y=1,Z=1))),(Item=(Type=Bone,Name=\"dyn_hair_back_middle_02\"),Transform=(Rotation=(X=0,Y=0,Z=0,W=1),Translation=(X=-1,Y=0,Z=0),S ...(truncated)"
    WeightMinimum=0
    ItemToModify.1.Transform.Rotation=&RigVMFunction_MathQuaternionFromEuler(Euler=(X=0,Y=2,Z=2)).Result
    ItemToModify.0.Transform.Rotation=&RigVMFunction_MathQuaternionFromEuler(Euler=(X=0,Y=2,Z=2)).Result
    Weight=@n2.OutputParam
   )
   n3 = &RigUnit_SphericalPoseReader(
    ActiveRegionScaleFactors=(PositiveWidth=0.5,NegativeWidth=0.5,PositiveHeight=0.5,NegativeHeight=0.5)
    ActiveRegionSize=0.1
    Debug=(bDrawDebug=false,bDraw2D=false,bDrawLocalAxes=false,DebugScale=25,DebugSegments=20,DebugThickness=0.25)
    DriverAxis=(X=0,Y=-1,Z=0)
    DriverItem=(Type=Bone,Name="head")
    FalloffRegionScaleFactors=(PositiveWidth=0.5,NegativeWidth=0.5,PositiveHeight=0.5,NegativeHeight=0.5)
    FalloffSize=0.3
    FlipHeightScaling=False
    FlipWidthScaling=False
    OptionalParentItem=(Type=Bone,Name="None")
    RotationOffset=(X=-120,Y=0,Z=10)
    ExecutePin=@RigVMFunction_Sequence.C
   )
   Multiply = &RigVMFunction_MathDoubleMul(A=@n3.OutputParam, B=RigVMVariableNode() {variable=is melee float}.Value)
   Clamp = &RigVMFunction_MathDoubleClamp(Value=@Multiply.Result)
   > &RigUnit_ModifyTransforms(
    ItemToModify=((Item=(Type=Bone,Name="dyn_hair_back_right_MSTR"),Transform=(Rotation=(X=0,Y=0,Z=0,W=1),Translation=(X=-5,Y=0,Z=0),Scale3D=(X=1,Y=1,Z=1))))
    WeightMinimum=0
    Weight=@Clamp.Result
    ItemToModify.0.Transform.Rotation=&RigVMFunction_MathQuaternionFromEuler(Euler=(X=0,Y=0,Z=0)).Result
   )
   n4 = &RigUnit_SphericalPoseReader(
    ActiveRegionScaleFactors=(PositiveWidth=1,NegativeWidth=1,PositiveHeight=1,NegativeHeight=1)
    ActiveRegionSize=0.1
    Debug=(bDrawDebug=true,bDraw2D=false,bDrawLocalAxes=true,DebugScale=25,DebugSegments=20,DebugThickness=0.25)
    DriverAxis=(X=-1,Y=0,Z=0)
    DriverItem=(Type=Bone,Name="upperarm_r")
    FalloffRegionScaleFactors=(PositiveWidth=1,NegativeWidth=1,PositiveHeight=1,NegativeHeight=1)
    FalloffSize=0.3
    FlipHeightScaling=False
    FlipWidthScaling=False
    OptionalParentItem=(Type=Bone,Name="spine_04")
    RotationOffset=(X=-130,Y=-80,Z=80)
    ExecutePin=@RigVMFunction_Sequence.D
   )
   &RigVMFunction_MathDoubleClamp(Value=&RigVMFunction_MathDoubleMul(A=@n4.OutputParam, B=RigVMVariableNode() {variable=is targeting float}.Value).Result)
   &RigVMFunction_ControlFlowBranch(ExecuteContext=(), Condition=RigVMVariableNode() {variable=is melee}.Value)
  ]}
]
hierarchy [
 root {loc=<0 0 0>, rot=<0 0 0 1>}
 pelvis {parent=root, loc=<0 -0.5301 98.0571>, rot=<0.0127 0.707 -0.0127 -0.707>}
 spine_01 {parent=pelvis, loc=<6.1208 -0.9328 0>, rot=<0 0 0.034 -0.9994>}
 spine_02 {parent=spine_01, loc=<6.5996 -0.2095 0>, rot=<0 0 0.0238 -0.9997>}
 spine_03 {parent=spine_02, loc=<6.4914 -0.1237 0>, rot=<0 0 0.0149 -0.9999>}
 spine_04 {parent=spine_03, loc=<6.9598 0.0794 0>, rot=<0 0 -0.0132 -0.9999>}
 spine_05 {parent=spine_04, loc=<6.9987 0 0>, rot=<0 0 -0.0015 -1>}
 neck_01 {parent=spine_05, loc=<11.3138 0.1529 0>, rot=<0 0 -0.2003 -0.9797>}
 neck_02 {parent=neck_01, loc=<4.5627 0 0>, rot=<0 0 0.0504 -0.9987>}
 head {parent=neck_02, loc=<5.3897 0.0009 -0.0006>, rot=<0 0 0.1149 -0.9934>}
 dyn_earring_MSTR {parent=head, loc=<-0.3006 0.2129 -6.5673>, rot=<-0.0209 -0.0004 0.9996 -0.0183>}
 dyn_earring_01 {parent=dyn_earring_MSTR, loc=<0 0 0>, rot=<0 0 0 1>}
 dyn_earring_02 {parent=dyn_earring_01, loc=<1.337 0 0>, rot=<0 0 0 1>}
 dyn_hair_left_MSTR {parent=head, loc=<4.2174 -2.3357 -6.2912>, rot=<0.0945 -0.4722 0.8748 -0.0536>}
 dyn_hair_left_01 {parent=dyn_hair_left_MSTR, loc=<0 0 0>, rot=<0 0 0 1>}
 dyn_hair_left_02 {parent=dyn_hair_left_01, loc=<7.2051 0 0>, rot=<0.2583 -0.3067 -0.0579 -0.9143>}
 dyn_hair_left_03 {parent=dyn_hair_left_02, loc=<3.704 0 0>, rot=<-0.1568 -0.1776 0.0712 -0.9689>}
 dyn_hair_small_l_01 {parent=dyn_hair_left_02, loc=<0.724 3.0105 0.6063>, rot=<-0.2554 0.4217 -0.4364 -0.7527>}
 dyn_hair_right_MSTR {parent=head, loc=<5.09 1.8614 7.2292>, rot=<-0.0578 -0.7047 -0.7025 -0.0808>}
 dyn_hair_right_01 {parent=dyn_hair_right_MSTR, loc=<0 0 0>, rot=<0 0 0 1>}
 dyn_hair_right_02 {parent=dyn_hair_right_01, loc=<6.8284 0 0>, rot=<0.0002 -0.002 0.0482 -0.9988>}
 dyn_hair_right_03 {parent=dyn_hair_right_02, loc=<5.8615 0 0>, rot=<0.007 -0.0148 0.3047 -0.9523>}
 dyn_hair_right_04 {parent=dyn_hair_right_03, loc=<3.5997 0 0>, rot=<-0.0416 0.0646 -0.3285 -0.9414>}
 dyn_hair_small_r_01 {parent=dyn_hair_right_01, loc=<2.3101 1.7966 0.8276>, rot=<-0.2539 0.0468 -0.0783 -0.9629>}
 dyn_hair_bang_r_01 {parent=dyn_hair_right_01, loc=<0.5672 -1.1835 6.088>, rot=<-0.0105 -0.1081 -0.1287 -0.9857>}
 dyn_hair_bang_r_02 {parent=dyn_hair_bang_r_01, loc=<6.857 0 0>, rot=<-0.1901 0.2617 -0.2014 -0.9246>}
 dyn_hair_back_MSTR_MSTR {parent=head, loc=<10.6544 -3.8764 1.8228>, rot=<0.0068 0.1801 0.9835 -0.0126>}
 dyn_hair_back_right_MSTR {parent=dyn_hair_back_MSTR_MSTR, loc=<3.1396 -0.3156 4.1185>, rot=<-0.367 0.0132 0.0052 -0.9301>}
 dyn_hair_back_right_01 {parent=dyn_hair_back_right_MSTR, loc=<0 0 0>, rot=<0 0 0 1>}
 dyn_hair_back_right_02 {parent=dyn_hair_back_right_01, loc=<12.1972 -0.1148 -0.1216>, rot=<0 0 0 1>}
 dyn_hair_back_middle_MSTR {parent=dyn_hair_back_MSTR_MSTR, loc=<2.5075 1.2648 0>, rot=<0 0 0 1>}
 dyn_hair_back_middle_01 {parent=dyn_hair_back_middle_MSTR, loc=<0 0 0>, rot=<0 0 0 1>}
 dyn_hair_back_middle_02 {parent=dyn_hair_back_middle_01, loc=<10.0899 -0.0591 -0.0359>, rot=<0 0 0 1>}
 dyn_hair_back_left_MSTR {parent=dyn_hair_back_MSTR_MSTR, loc=<2.6007 0.9815 -4.8373>, rot=<0.2847 -0.0063 0.0019 -0.9586>}
 dyn_hair_back_left_01 {parent=dyn_hair_back_left_MSTR, loc=<0 0 0>, rot=<0 0 0 1>}
 dyn_hair_back_left_02 {parent=dyn_hair_back_left_01, loc=<10.0904 -0.0299 -0.0222>, rot=<0 0 0 1>}
 clavicle_l {parent=spine_05, loc=<5.4831 -0.1053 -3.2946>, rot=<0.0465 -0.7289 -0.009 -0.6829>}
 upperarm_l {parent=clavicle_l, loc=<9.8658 -1.2136 -0.4105>, rot=<0.0164 -0.3245 -0.0269 -0.9454>}
 clavicle_r {parent=spine_05, loc=<5.4831 -0.1053 3.2946>, rot=<0.7289 0.0465 -0.6829 0.009>}
 upperarm_r {parent=clavicle_r, loc=<-9.8658 1.2136 0.4105>, rot=<0.0164 -0.3245 -0.0269 -0.9454>}
]
variables [
 is melee {type=bool, default=False}
 is melee float {type=double, default=0}
 is targeting float {type=double, default=0}
 not skydiving {type=bool, default=False}
]
