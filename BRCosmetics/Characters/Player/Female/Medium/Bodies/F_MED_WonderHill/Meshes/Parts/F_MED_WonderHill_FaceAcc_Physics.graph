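# Physics asset for the face accessory mesh of the WonderHill medium female player body, setting up two dynamic hair chains.
# Constraints with tight twist and moderate swing limits link capsules into a three-segment left and four-segment right chain, each rooted at a small MSTR sphere.
# Unconstrained capsules on head, neck, upper spine, both clavicles and only the right upper arm provide body collision shapes.
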PhysicsAsset /BRCosmetics/Characters/Player/Female/Medium/Bodies/F_MED_WonderHill/Meshes/Parts/F_MED_WonderHill_FaceAcc_Physics
bodies {
 clavicle_l [{
   shapeType=Capsule
   length=10
   radius=4
   shapeName=clavicle_r_capsule
   center=<6.0046 0.1558 0.1762>
   rotation=<rot -75.4384 179.7274 -5.1392>
  }]
 clavicle_r [{
   shapeType=Capsule
   length=10
   radius=4
   shapeName=clavicle_r_capsule
   center=<-5.8705 -0.1564 0.3403>
   rotation=<rot -77.2207 -179.548 174.1567>
  }]
 dyn_hair_left_01 [{shapeType=Capsule, length=3, radius=2.9, shapeName=dyn_hair_left_01_capsule, center=<2.2057 0 0>, rotation=<rot -8.851 -0.0001 0>}]
 dyn_hair_left_02 [{shapeType=Capsule, length=3, radius=2.15, shapeName=dyn_hair_left_02_capsule, center=<1.367 0 0.7433>, rotation=<rot 28.536 0 0>}]
 dyn_hair_left_03 [{
   shapeType=Capsule
   length=4.7829
   radius=2.1247
   shapeName=dyn_hair_left_03_capsule
   center=<3.4827 -0.8504 -0.1001>
   rotation=<rot 66.6541 -7.5132 17.3573>
  }]
 dyn_hair_left_MSTR [{shapeType=Sphere, radius=1, shapeName=dyn_hair_left_MSTR_sphere}]
 dyn_hair_right_01 [{shapeType=Capsule, length=3, radius=1.5592, shapeName=dyn_hair_right_01_capsule, center=<3.3026 0 0>, rotation=<rot 90 0 0>}]
 dyn_hair_right_02 [{shapeType=Capsule, length=3, radius=1.8, shapeName=dyn_hair_right_02_capsule, center=<2.9058 0 0>, rotation=<rot 90 -172.5953 -161.139>}]
 dyn_hair_right_03 [{
   shapeType=Capsule
   length=3.459
   radius=3.75
   shapeName=dyn_hair_right_03_capsule
   center=<2.8093 0.1025 0.1088>
   rotation=<rot 58.8072 -179.9983 -179.998>
  }]
 dyn_hair_right_04 [{shapeType=Capsule, length=4, radius=3, shapeName=dyn_hair_right_04_capsule, center=<3.3178 0 0.1764>, rotation=<rot 76.7837 -180 -180>}]
 dyn_hair_right_MSTR [{shapeType=Sphere, radius=1, shapeName=dyn_hair_right_MSTR_sphere}]
 head [{shapeType=Capsule, length=3, radius=5.5, shapeName=head_capsule, center=<0.3726 3.3722 0>, rotation=<rot 90 0 0>}]
 neck_01 [{shapeType=Capsule, length=7.2244, radius=5, shapeName=neck_01_capsule, center=<-5.547 2.3355 0>}]
 neck_02 [{shapeType=Capsule, length=3, radius=4, shapeName=neck_02_capsule, center=<-2.4907 0.5442 0>, rotation=<rot 90 0 0>}]
 spine_05 [{shapeType=Capsule, length=15.8358, radius=10.6902, shapeName=spine_05_capsule, center=<-1.4655 -2.3176 0>}]
 upperarm_r [{
   shapeType=Capsule
   length=20
   radius=2.3
   shapeName=upperarm_r_capsule
   center=<-11.579 -0.0008 0.2181>
   rotation=<rot -88.9211 -179.998 -179.9979>
  }]
}
constraints [
 {bone1Name=dyn_hair_right_04, bone2Name=dyn_hair_right_03, swing1LimitDegrees=25, swing2LimitDegrees=30, twistLimitDegrees=5}
 {bone1Name=dyn_hair_right_03, bone2Name=dyn_hair_right_02, swing1LimitDegrees=25, swing2LimitDegrees=30, twistLimitDegrees=5}
 {bone1Name=dyn_hair_right_02, bone2Name=dyn_hair_right_01, swing1LimitDegrees=15, swing2LimitDegrees=35, twistLimitDegrees=5}
 {bone1Name=dyn_hair_right_01, bone2Name=dyn_hair_right_MSTR, swing1LimitDegrees=10, swing2LimitDegrees=25, twistLimitDegrees=5}
 {bone1Name=dyn_hair_left_03, bone2Name=dyn_hair_left_02, swing1LimitDegrees=20, swing2LimitDegrees=20, twistLimitDegrees=5}
 {bone1Name=dyn_hair_left_02, bone2Name=dyn_hair_left_01, swing1LimitDegrees=20, swing2LimitDegrees=15, twistLimitDegrees=2}
 {bone1Name=dyn_hair_left_01, bone2Name=dyn_hair_left_MSTR, swing1LimitDegrees=10, swing2LimitDegrees=10, twistLimitDegrees=2}
]
