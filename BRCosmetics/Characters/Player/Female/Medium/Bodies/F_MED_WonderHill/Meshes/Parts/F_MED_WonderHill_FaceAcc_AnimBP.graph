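# Post-processes hair and earring motion for the WonderHill female character's face accessory mesh.
# Copies the parent pose through a Control Rig, RigidBody, an earring-to-head constraint and AnimDynamics, then caches it as 'dynamics'.
# Layered bone offsets and scales then correct that pose for crouching, aiming, melee, skydive floating, parachuting, DBNO and front end.
# The EventGraph derives blend alphas from state flags and bone rotation deltas on the customization base mesh.
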AnimBlueprint /BRCosmetics/Characters/Player/Female/Medium/Bodies/F_MED_WonderHill/Meshes/Parts/F_MED_WonderHill_FaceAcc_AnimBP
target_skeleton=&/F_MED_WonderHill_FaceAcc_Skeleton
graphs [
 AnimGraph {nodes [
   "Anim Dynamics - Chain: dyn_hair_back_middle_01 -> dyn_hair_back_middle_02"(Alpha=0.7, ExternalForce="0,0,0") {
    angularDampingOverride=1
    angularSpringConstant=1500
    bAngularSpring=true
    bChain=true
    bGravityOverrideInSimSpace=true
    bOverrideAngularDamping=true
    bUseGravityOverride=true
    gravityScale=0.35
    alphaBoolBlend {blendOption=Linear}
    boundBone {boneName=dyn_hair_back_middle_01}
    chainEnd {boneName=dyn_hair_back_middle_02}
    componentAppliedLinearAccClamp=<2000 2000 2500>
    componentLinearAccScale=<0.1 0.1 0.4>
    componentLinearVelScale=<0.2 0.2 0.45>
    gravityOverride {z=-550}
    physicsBodyDefinitions [
     {
      sphereCollisionRadius=0
      boundBone {boneName=dyn_hair_back_middle_01}
      boxExtents=<18 12 12>
      constraintSetup {angularLimitsMax=<0 5 15>, angularLimitsMin=<0 -5 0>, angularTarget=<0 0 0>}
      localJointOffset=<3 0 0>
     }
     {
      sphereCollisionRadius=0
      boundBone {boneName=dyn_hair_back_middle_02}
      boxExtents=<20 20 12>
      constraintSetup {angularLimitsMax=<0 10 50>, angularLimitsMin=<0 -10 0>, angularTarget=<0 0 0>}
      localJointOffset=<4 0 0>
     }
    ]
    relativeSpaceBone {boneName=root}
   }
   n1 = "Transform (Modify) Bone - Bone: dyn_hair_back_right_MSTR"(
    ComponentPose=Local To Component(LocalPose=Use cached pose 'dynamics'())
    Alpha=Get head rotating right when melee() -> (head rotating right when melee:double)
    Translation="0,2,-2"
    Rotation="0,0,0"
    Scale="1,1,1"
   ) {
    rotationMode=BMM_Additive
    rotationSpace=BCS_BoneSpace
    translationMode=BMM_Additive
    translationSpace=BCS_BoneSpace
    alphaBoolBlend {blendInTime=0.2, blendOption=QuadraticInOut, blendOutTime=0.5}
    boneToModify {boneName=dyn_hair_back_right_MSTR}
    translation {y=2, z=-2}
   }
   n2 = Get crouching moving() -> (crouching moving:bool)
   n3 = "Transform (Modify) Bone - Bone: dyn_hair_back_left_02"!(ComponentPose=@n1, bAlphaBoolEnabled:bool=@n2, Translation="-1,0,0", Rotation="0,3,0", Scale="1,1,1") -> (Pose) {
    alphaInputType=Bool
    rotationMode=BMM_Additive
    rotationSpace=BCS_BoneSpace
    translationMode=BMM_Additive
    translationSpace=BCS_BoneSpace
    alphaBoolBlend {blendInTime=0.2, blendOption=Linear, blendOutTime=0.2}
    boneToModify {boneName=dyn_hair_back_left_02}
    rotation {yaw=3}
    translation {x=-1}
   }
   n4 = "Transform (Modify) Bone - Bone: dyn_hair_back_middle_02"!(ComponentPose=@n3, bAlphaBoolEnabled:bool=@n2, Translation="-1,0,0", Rotation="0,3,0", Scale="1,1,1") -> (Pose) {
    alphaInputType=Bool
    rotationMode=BMM_Additive
    rotationSpace=BCS_BoneSpace
    translationMode=BMM_Additive
    translationSpace=BCS_BoneSpace
    alphaBoolBlend {blendInTime=0.2, blendOption=Linear, blendOutTime=0.2}
    boneToModify {boneName=dyn_hair_back_middle_02}
    rotation {yaw=3}
    translation {x=-1}
   }
   n5 = Get bIsParachuteOpened()
   n6 = "Transform (Modify) Bone - Bone: dyn_hair_back_right_MSTR"!(ComponentPose=@n4, bAlphaBoolEnabled:bool=@n5, Translation="0,0,0", Rotation="0,-5,0", Scale="1,1,1") -> (Pose) {
    alphaInputType=Bool
    rotationMode=BMM_Additive
    rotationSpace=BCS_BoneSpace
    translationMode=BMM_Additive
    translationSpace=BCS_BoneSpace
    alphaBoolBlend {blendInTime=0.2, blendOption=Linear, blendOutTime=0.2}
    boneToModify {boneName=dyn_hair_back_right_MSTR}
    rotation {yaw=-5}
   }
   n7 = "Transform (Modify) Bone - Bone: dyn_hair_left_02"!(ComponentPose=@n6, bAlphaBoolEnabled:bool=@n5, Translation="-1,0,0", Rotation="0,0,20", Scale="1,1,1") -> (Pose) {
    alphaInputType=Bool
    rotationMode=BMM_Additive
    rotationSpace=BCS_BoneSpace
    translationMode=BMM_Additive
    translationSpace=BCS_BoneSpace
    alphaBoolBlend {blendInTime=0.2, blendOption=Linear, blendOutTime=0.2}
    boneToModify {boneName=dyn_hair_left_02}
    rotation {roll=20}
    translation {x=-1}
   }
   n8 = "Transform (Modify) Bone - Bone: dyn_hair_small_l_01"!(ComponentPose=@n7, bAlphaBoolEnabled:bool=@n5, Translation="0,0,0", Rotation="-10,0,0", Scale="1,1,1") -> (Pose) {
    alphaInputType=Bool
    rotationMode=BMM_Additive
    rotationSpace=BCS_BoneSpace
    translationMode=BMM_Additive
    translationSpace=BCS_BoneSpace
    alphaBoolBlend {blendInTime=0.2, blendOption=Linear, blendOutTime=0.2}
    boneToModify {boneName=dyn_hair_small_l_01}
    rotation {pitch=-10}
   }
   n9 = Get bIsSkydiveFloating()
   n10 = "Transform (Modify) Bone - Bone: dyn_hair_back_middle_MSTR"!(ComponentPose=@n8, bAlphaBoolEnabled:bool=@n9, Translation="0,0,0", Rotation="0,5,0", Scale="1,1,1") -> (Pose) {
    alphaInputType=Bool
    rotationMode=BMM_Additive
    rotationSpace=BCS_BoneSpace
    translationMode=BMM_Additive
    translationSpace=BCS_BoneSpace
    alphaBoolBlend {blendInTime=0.2, blendOption=Linear, blendOutTime=0.2}
    boneToModify {boneName=dyn_hair_back_middle_MSTR}
    rotation {yaw=5}
   }
   n11 = "Transform (Modify) Bone - Bone: dyn_hair_back_left_MSTR"!(ComponentPose=@n10, bAlphaBoolEnabled:bool=@n9, Translation="-1,0,0", Rotation="0,5,0", Scale="1,1,1") -> (Pose) {
    alphaInputType=Bool
    rotationMode=BMM_Additive
    rotationSpace=BCS_BoneSpace
    translationMode=BMM_Additive
    translationSpace=BCS_BoneSpace
    alphaBoolBlend {blendInTime=0.2, blendOption=Linear, blendOutTime=0.2}
    boneToModify {boneName=dyn_hair_back_left_MSTR}
    rotation {yaw=5}
    translation {x=-1}
   }
   n12 = "Transform (Modify) Bone - Bone: dyn_hair_back_middle_02"(ComponentPose=@n11, bAlphaBoolEnabled=@n9, Translation="-1,1,0", Rotation="0,0,0", Scale="1,1,1") {
    alphaInputType=Bool
    rotationMode=BMM_Additive
    rotationSpace=BCS_BoneSpace
    translationMode=BMM_Additive
    translationSpace=BCS_BoneSpace
    alphaBoolBlend {blendInTime=0.2, blendOption=Linear, blendOutTime=0.2}
    boneToModify {boneName=dyn_hair_back_middle_02}
    translation=<-1 1>
   }
   n13 = "Transform (Modify) Bone - Bone: dyn_hair_back_left_02"(ComponentPose=@n12, bAlphaBoolEnabled=@n9, Translation="-1,1,0", Rotation="0,3,0", Scale="1,1,1") {
    alphaInputType=Bool
    rotationMode=BMM_Additive
    rotationSpace=BCS_BoneSpace
    translationMode=BMM_Additive
    translationSpace=BCS_BoneSpace
    alphaBoolBlend {blendInTime=0.2, blendOption=Linear, blendOutTime=0.2}
    boneToModify {boneName=dyn_hair_back_left_02}
    rotation {yaw=3}
    translation=<-1 1>
   }
   n14 = "Transform (Modify) Bone - Bone: dyn_hair_small_l_01"(ComponentPose=@n13, bAlphaBoolEnabled=@n9, Translation="0,0,0", Rotation="-5,0,0", Scale="1,1,1") {
    alphaInputType=Bool
    rotationMode=BMM_Additive
    rotationSpace=BCS_BoneSpace
    translationMode=BMM_Additive
    translationSpace=BCS_BoneSpace
    alphaBoolBlend {blendInTime=0.2, blendOption=Linear, blendOutTime=0.2}
    boneToModify {boneName=dyn_hair_small_l_01}
    rotation {pitch=-5}
   }
   n15 = "Transform (Modify) Bone - Bone: dyn_earring_MSTR"(
    ComponentPose=@n14
    Alpha=Get earring rotation alpha() -> (earring rotation alpha:float)
    Translation="0,0,0"
    Rotation="-25,-25,0"
    Scale="1,1,1"
   ) {
    rotationMode=BMM_Additive
    rotationSpace=BCS_BoneSpace
    translationMode=BMM_Additive
    translationSpace=BCS_BoneSpace
    alphaBoolBlend {blendOption=Linear}
    boneToModify {boneName=dyn_earring_MSTR}
    rotation {pitch=-25, yaw=-25}
   }
   n16 = Get crouching looking up() -> (crouching looking up:double)
   n17 = "Transform (Modify) Bone - Bone: dyn_hair_back_left_MSTR"!(ComponentPose=@n15, Alpha:float=@n16, Translation="-1,0,0", Rotation="0,0,0", Scale="1,1,1") -> (Pose) {
    rotationMode=BMM_Additive
    rotationSpace=BCS_BoneSpace
    translationMode=BMM_Additive
    translationSpace=BCS_BoneSpace
    alphaBoolBlend {blendOption=Linear}
    boneToModify {boneName=dyn_hair_back_left_MSTR}
    translation {x=-1}
   }
   n18 = "Transform (Modify) Bone - Bone: dyn_hair_left_MSTR"(ComponentPose=@n17, Alpha=@n16, Translation="-2,0,0", Rotation="0,0,0", Scale="1,1,1") {
    rotationMode=BMM_Additive
    rotationSpace=BCS_BoneSpace
    translationMode=BMM_Additive
    translationSpace=BCS_BoneSpace
    alphaBoolBlend {blendOption=Linear}
    boneToModify {boneName=dyn_hair_left_MSTR}
    translation {x=-2}
   }
   n19 = "Transform (Modify) Bone - Bone: dyn_hair_left_MSTR"!(ComponentPose=@n18, bAlphaBoolEnabled:bool=Get bIsDBNO(), Translation="0,0,0", Rotation="13,13,0", Scale="1,1,1") -> (Pose) {
    alphaInputType=Bool
    rotationMode=BMM_Additive
    rotationSpace=BCS_BoneSpace
    translationMode=BMM_Additive
    translationSpace=BCS_BoneSpace
    alphaBoolBlend {blendInTime=0.2, blendOption=Linear, blendOutTime=0.2}
    boneToModify {boneName=dyn_hair_left_MSTR}
    rotation {pitch=13, yaw=13}
   }
   n20 = Get targeting up() -> (targeting up:double)
   n21 = "Transform (Modify) Bone - Bone: dyn_hair_right_02"(ComponentPose=@n19, Alpha=@n20, Translation="0,0,0", Rotation="0,0,0", Scale="0.7,0.5,1") {
    rotationMode=BMM_Additive
    rotationSpace=BCS_BoneSpace
    scaleMode=BMM_Replace
    scaleSpace=BCS_BoneSpace
    translationMode=BMM_Additive
    translationSpace=BCS_BoneSpace
    alphaBoolBlend {blendInTime=0.2, blendOption=Linear, blendOutTime=0.2}
    alphaScaleBiasClamp {bInterpResult=true, interpSpeedDecreasing=15, interpSpeedIncreasing=2}
    boneToModify {boneName=dyn_hair_right_02}
    scale=<0.7 0.5>
   }
   n22 = "Transform (Modify) Bone - Bone: dyn_hair_right_03"(ComponentPose=@n21, Alpha=@n20, Translation="0,0,0", Rotation="0,0,0", Scale="1,0.6,1") {
    rotationMode=BMM_Additive
    rotationSpace=BCS_BoneSpace
    scaleMode=BMM_Replace
    scaleSpace=BCS_BoneSpace
    translationMode=BMM_Additive
    translationSpace=BCS_BoneSpace
    alphaBoolBlend {blendInTime=0.2, blendOption=Linear, blendOutTime=0.2}
    alphaScaleBiasClamp {bInterpResult=true, interpSpeedDecreasing=15, interpSpeedIncreasing=2}
    boneToModify {boneName=dyn_hair_right_03}
    scale {y=0.6}
   }
   n23 = "Transform (Modify) Bone - Bone: dyn_hair_right_02"(
    ComponentPose=@n22
    Alpha=Get targeting up and rotationg() -> (targeting up and rotationg:double)
    Translation="0,0,2"
    Rotation="0,15,0"
    Scale="1,1,1"
   ) {
    rotationMode=BMM_Additive
    rotationSpace=BCS_BoneSpace
    scaleSpace=BCS_BoneSpace
    translationMode=BMM_Additive
    translationSpace=BCS_BoneSpace
    alphaBoolBlend {blendInTime=0.2, blendOption=Linear, blendOutTime=0.2}
    alphaScaleBiasClamp {bInterpResult=true, interpSpeedDecreasing=15, interpSpeedIncreasing=2}
    boneToModify {boneName=dyn_hair_right_02}
    rotation {yaw=15}
    translation {z=2}
   }
   n24 = Get bIsInFrontEnd()
   n25 = "Transform (Modify) Bone - Bone: dyn_hair_right_02"!(ComponentPose=@n23, bAlphaBoolEnabled:bool=@n24, Translation="0,0,0", Rotation="6,0,0", Scale="1,1,1") -> (Pose) {
    alphaInputType=Bool
    rotationMode=BMM_Additive
    rotationSpace=BCS_BoneSpace
    scaleSpace=BCS_BoneSpace
    translationMode=BMM_Additive
    translationSpace=BCS_BoneSpace
    alphaBoolBlend {blendInTime=0.2, blendOption=Linear, blendOutTime=0.2}
    alphaScaleBiasClamp {bInterpResult=true, interpSpeedDecreasing=15, interpSpeedIncreasing=2}
    boneToModify {boneName=dyn_hair_right_02}
    rotation {pitch=6}
   }
   n26 = "Transform (Modify) Bone - Bone: dyn_hair_left_MSTR"(ComponentPose=@n25, bAlphaBoolEnabled=@n24, Translation="-1,0,0", Rotation="0,0,0", Scale="1,1,1") {
    alphaInputType=Bool
    rotationMode=BMM_Additive
    rotationSpace=BCS_BoneSpace
    scaleSpace=BCS_BoneSpace
    translationMode=BMM_Additive
    translationSpace=BCS_BoneSpace
    alphaBoolBlend {blendInTime=0.2, blendOption=Linear, blendOutTime=0.2}
    alphaScaleBiasClamp {bInterpResult=true, interpSpeedDecreasing=15, interpSpeedIncreasing=2}
    boneToModify {boneName=dyn_hair_left_MSTR}
    translation {x=-1}
   }
   Output Pose(Result=Component To Local(ComponentPose=@n26)) {layerGroup=None}
   n27 = Copy Pose From Mesh() {bUseAttachedParent=true}
   n28 = Control Rig!(
    Source=Component To Local(ComponentPose=Local To Component(LocalPose=@n27))
    Alpha
    not skydiving:bool=Get not sky diving() -> (not sky diving:bool)
    is melee:bool=Get bIsPlayingMeleeAnim()
    is melee float:double=Get is melee float() -> (is melee float:double)
    is targeting float:double=Get is targeting float() -> (is targeting float:double)
   ) -> (Pose) {
    alphaBoolBlend {blendOption=Linear}
    controlRigAssetReference {blueprintRigClass=&/F_MED_WonderHill_FaceAcc_CtrlRig.F_MED_WonderHill_FaceAcc_CtrlRig_C}
   }
   n29 = RigidBody(ComponentPose=Local To Component(LocalPose=@n28)) {
    bOverrideWorldGravity=true
    overridePhysicsAsset=&/F_MED_WonderHill_FaceAcc_Physics
    simulationSpace=BaseBoneSpace
    alphaBoolBlend {blendOption=Linear}
    baseBoneRef {boneName=spine_05}
    componentAppliedLinearAccClamp=<500 500 500>
    componentLinearAccScale=<0.1 0.1 0.1>
    componentLinearVelScale=<0.2 0.2 0.2>
    overrideWorldGravity {z=-980}
   }
   n30 = "Constraint - dyn_earring_MSTR (P)"!(ComponentPose=@n29, Alpha=1, ConstraintWeights_0=1, ConstraintWeights_1:float=0) -> (Pose) {
    alphaBoolBlend {blendOption=Linear}
    boneToModify {boneName=dyn_earring_MSTR}
    constraintSetup [{targetBone {boneName=head}}]
    constraintWeights [1,0]
   }
   n31 = Anim Dynamics - Bone: dyn_earring_01(ComponentPose=@n30, Alpha=1, ExternalForce="0,0,0") {
    angularDampingOverride=0.9999
    bGravityOverrideInSimSpace=true
    bOverrideAngularDamping=true
    simulationSpace=RootRelative
    alphaBoolBlend {blendOption=Linear}
    boundBone {boneName=dyn_earring_01}
    physicsBodyDefinitions [{
      sphereCollisionRadius=0
      boundBone {boneName=dyn_earring_01}
      boxExtents=<6 15 15>
      constraintSetup {angularLimitsMax=<25 75 45>, angularLimitsMin=<-25 -10 -45>, angularTarget=<0 0 0>}
      localJointOffset=<1.5 -0.05 0.1>
     }]
    sphericalLimits [{limitRadius=4, drivingBone {boneName=head}, sphereLocalOffset=<-2 0 0>}]
   }
   n32 = Anim Dynamics - Bone: dyn_hair_small_r_01(ComponentPose=@n31, Alpha=1, ExternalForce="0,0,0") {
    angularDampingOverride=0.9999
    angularSpringConstant=1000
    bAngularSpring=true
    bGravityOverrideInSimSpace=true
    bOverrideAngularDamping=true
    gravityScale=0.1
    simulationSpace=RootRelative
    alphaBoolBlend {blendOption=Linear}
    boundBone {boneName=dyn_hair_small_r_01}
    componentAppliedLinearAccClamp=<3000 3000 3000>
    componentLinearAccScale=<0.5 0.5 0.5>
    componentLinearVelScale=<0.6 0.6 0.6>
    gravityOverride {z=-980}
    physicsBodyDefinitions [{
      sphereCollisionRadius=0
      boundBone {boneName=dyn_hair_small_r_01}
      boxExtents=<14 12 17>
      constraintSetup {angularLimitsMax=<10 0 50>, angularLimitsMin=<-10 -40 -10>, angularTarget=<0 0 0>}
      localJointOffset=<4 0 0>
     }]
    relativeSpaceBone {boneName=head}
   }
   n33 = "Anim Dynamics - Chain: dyn_hair_bang_r_01 -> dyn_hair_bang_r_02"(ComponentPose=@n32, Alpha=1, ExternalForce="0,0,0") {
    angularDampingOverride=0.9999
    angularSpringConstant=7000
    bAngularSpring=true
    bChain=true
    bGravityOverrideInSimSpace=true
    bOverrideAngularDamping=true
    bUseGravityOverride=true
    simulationSpace=RootRelative
    alphaBoolBlend {blendOption=Linear}
    boundBone {boneName=dyn_hair_bang_r_01}
    chainEnd {boneName=dyn_hair_bang_r_02}
    componentAppliedLinearAccClamp=<3000 3000 100>
    componentLinearAccScale=<0.2 0.4 0.4>
    componentLinearVelScale=<0.2 0.5 0.4>
    gravityOverride {z=-400}
    physicsBodyDefinitions [
     {
      sphereCollisionRadius=0
      boundBone {boneName=dyn_hair_bang_r_01}
      boxExtents=<15 16 14>
      constraintSetup {angularLimitsMax=<0 0 8>, angularLimitsMin=<0 -20 0>, angularTarget=<0 0 0>}
      localJointOffset=<0 0 0>
     }
     {
      sphereCollisionRadius=0
      boundBone {boneName=dyn_hair_bang_r_02}
      boxExtents=<10 12 11>
      constraintSetup {angularLimitsMax=<0 0 50>, angularLimitsMin=<0 -20 0>, angularTarget=<0 0 0>}
      localJointOffset=<0 0 0>
     }
    ]
   }
   n34 = Anim Dynamics - Bone: dyn_hair_small_l_01(ComponentPose=@n33, Alpha=1, ExternalForce="0,0,0") {
    angularDampingOverride=0.9999
    angularSpringConstant=1000
    bAngularSpring=true
    bGravityOverrideInSimSpace=true
    bOverrideAngularDamping=true
    gravityScale=0.15
    simulationSpace=RootRelative
    alphaBoolBlend {blendOption=Linear}
    boundBone {boneName=dyn_hair_small_l_01}
    componentAppliedLinearAccClamp=<3000 300 3000>
    componentLinearAccScale=<0.6 0.3 0.3>
    componentLinearVelScale=<0.7 0.4 0.4>
    gravityOverride {z=-980}
    physicsBodyDefinitions [{
      sphereCollisionRadius=0
      boundBone {boneName=dyn_hair_small_l_01}
      boxExtents=<14 12 17>
      constraintSetup {angularLimitsMax=<0 30 50>, angularLimitsMin=<0 -30 -10>, angularTarget=<0 0 0>}
      localJointOffset=<2 0 0>
     }]
    relativeSpaceBone {boneName=head}
   }
   n35 = "Anim Dynamics - Chain: dyn_hair_back_left_01 -> dyn_hair_back_left_02"(ComponentPose=@n34, Alpha=0.6, ExternalForce="0,0,0") {
    angularDampingOverride=1
    angularSpringConstant=1500
    bAngularSpring=true
    bChain=true
    bGravityOverrideInSimSpace=true
    bOverrideAngularDamping=true
    bUseGravityOverride=true
    gravityScale=0.35
    alphaBoolBlend {blendOption=Linear}
    boundBone {boneName=dyn_hair_back_left_01}
    chainEnd {boneName=dyn_hair_back_left_02}
    componentAppliedLinearAccClamp=<2000 2000 2500>
    componentLinearAccScale=<0.1 0.1 0.38>
    componentLinearVelScale=<0.2 0.2 0.43>
    gravityOverride {z=-550}
    physicsBodyDefinitions [
     {
      sphereCollisionRadius=0
      boundBone {boneName=dyn_hair_back_left_01}
      boxExtents=<18 12 12>
      constraintSetup {angularLimitsMax=<0 5 20>, angularLimitsMin=<0 -5 0>, angularTarget=<0 0 0>}
      localJointOffset=<3 0 0>
     }
     {
      sphereCollisionRadius=0
      boundBone {boneName=dyn_hair_back_left_02}
      boxExtents=<20 20 12>
      constraintSetup {angularLimitsMax=<0 10 50>, angularLimitsMin=<0 -10 0>, angularTarget=<0 0 0>}
      localJointOffset=<4 0 0>
     }
    ]
    relativeSpaceBone {boneName=root}
   }
   n36 = "Anim Dynamics - Chain: dyn_hair_back_right_01 -> dyn_hair_back_right_02"(ComponentPose=@n35, Alpha=0.6, ExternalForce="0,0,0") {
    angularDampingOverride=1
    angularSpringConstant=1500
    bAngularSpring=true
    bChain=true
    bGravityOverrideInSimSpace=true
    bOverrideAngularDamping=true
    bUseGravityOverride=true
    gravityScale=0.35
    alphaBoolBlend {blendOption=Linear}
    boundBone {boneName=dyn_hair_back_right_01}
    chainEnd {boneName=dyn_hair_back_right_02}
    componentAppliedLinearAccClamp=<2000 2000 1700>
    componentLinearAccScale=<0.1 0.1 0.38>
    componentLinearVelScale=<0.2 0.2 0.43>
    gravityOverride {z=-550}
    physicsBodyDefinitions [
     {
      sphereCollisionRadius=0
      boundBone {boneName=dyn_hair_back_right_01}
      boxExtents=<18 12 12>
      constraintSetup {angularLimitsMax=<0 5 20>, angularLimitsMin=<0 -5 0>, angularTarget=<0 0 0>}
      localJointOffset=<3 0 0>
     }
     {
      sphereCollisionRadius=0
      boundBone {boneName=dyn_hair_back_right_02}
      boxExtents=<20 20 12>
      constraintSetup {angularLimitsMax=<0 10 30>, angularLimitsMin=<0 -10 0>, angularTarget=<0 0 0>}
      localJointOffset=<4 0 0>
     }
    ]
    relativeSpaceBone {boneName=root}
   }
   n37 = "Anim Dynamics - Chain: dyn_hair_back_middle_01 -> dyn_hair_back_middle_02"(ComponentPose=@n36, Alpha=0.7, ExternalForce="0,0,0") {
    angularDampingOverride=1
    angularSpringConstant=1500
    bAngularSpring=true
    bChain=true
    bGravityOverrideInSimSpace=true
    bOverrideAngularDamping=true
    bUseGravityOverride=true
    gravityScale=0.35
    alphaBoolBlend {blendOption=Linear}
    boundBone {boneName=dyn_hair_back_middle_01}
    chainEnd {boneName=dyn_hair_back_middle_02}
    componentAppliedLinearAccClamp=<2000 2000 1600>
    componentLinearAccScale=<0.1 0.1 0.38>
    componentLinearVelScale=<0.2 0.2 0.43>
    gravityOverride {z=-550}
    physicsBodyDefinitions [
     {
      sphereCollisionRadius=0
      boundBone {boneName=dyn_hair_back_middle_01}
      boxExtents=<20 15 15>
      constraintSetup {angularLimitsMax=<0 5 20>, angularLimitsMin=<0 -5 0>, angularTarget=<0 0 0>}
      localJointOffset=<3 0 0>
     }
     {
      sphereCollisionRadius=0
      boundBone {boneName=dyn_hair_back_middle_02}
      boxExtents=<22 22 15>
      constraintSetup {angularLimitsMax=<0 10 12>, angularLimitsMin=<0 -10 0>, angularTarget=<0 0 0>}
      localJointOffset=<4 0 0>
     }
    ]
    relativeSpaceBone {boneName=root}
   }
   Save cached pose 'dynamics'(Pose=Component To Local(ComponentPose=@n37)) {cacheName=dynamics}
  ]}
 EventGraph {nodes [
   n1 = Event BlueprintUpdateAnimation()
   > Set delta time x(delta time x:float=@n1.DeltaTimeX) -> (Output_Get:float)
   > n2 = Sequence()
   > Set crouching moving(crouching moving:bool=AND Boolean(A=Get bIsCrouching(), B=Get bIsMoving2D())) -> (Output_Get:bool)
   > Set earring rotation alpha(earring rotation alpha:float=GetAlphaFromReferenceTransformDelta(SkeletalMeshComponent=Get Skel Mesh(), BoneName=head, BaseName=spine_05, InputRange=(X=-29,Y=-54), TwistAxis="1,0,0")) -> (Output_Get:float)
   n3 = Get Skel Mesh()
   > Set crouching looking up(crouching looking up:double=float * float(
     A=GetAlphaFromReferenceTransformDelta(SkeletalMeshComponent=@n3, BoneName=head, BaseName=spine_05, InputRange=(X=-25,Y=-54), TwistAxis="1,0,0")
     B:bool=Get bIsCrouching()
    ) {is=title K2Node_PromotableOperator}) -> (Output_Get:double)
   > Set not sky diving(not sky diving:bool=NOT Boolean(A=Get bIsSkyDiving())) -> (Output_Get:bool)
   n4 = Get Skel Mesh()
   > Set targeting up(targeting up:double=float * float(
     A=GetAlphaFromReferenceTransformDelta(SkeletalMeshComponent=@n4, BoneName=upperarm_r, BaseName=spine_05, DeltaComponentType=EulerY, InputRange=(X=-20,Y=-4), TwistAxis="1,0,0")
     B=Get bIsTargeting()
    )) -> (Output_Get:double)
   n5 = GetAlphaFromReferenceTransformDelta(
    SkeletalMeshComponent=Get Skel Mesh()
    BoneName=spine_05
    BaseName=pelvis
    DeltaComponentType=EulerX
    InputRange=(X=-16,Y=5)
    TwistAxis="1,0,0"
   )
   n6 = Get targeting up()
   > Set targeting up and rotationg(targeting up and rotationg:double="Clamp (Float)"(Value=float * float(A=@n5, B:double=@n6), Min=0, Max=1)) -> (Output_Get:double)
   n7 = Get bIsPlayingMeleeAnim()
   > Set is melee float(is melee float:double=FInterpTo(Current=Get is melee float(), Target="To Float (Boolean)"(InBool=@n7), DeltaTime=Get delta time x() -> (delta time x:float), InterpSpeed=10)) -> (Output_Get:double)
   n8 = Get Skel Mesh()
   > Set head rotating right when melee(head rotating right when melee:double=float * float(
     A=GetAlphaFromReferenceTransformDelta(SkeletalMeshComponent=@n8, BoneName=head, BaseName=spine_05, DeltaComponentType=EulerY, InputRange=(X=0,Y=15), TwistAxis="1,0,0")
     B:bool=Get bIsPlayingMeleeAnim()
    )) -> (Output_Get:double)
   n9 = Branch(execute=@n2.then_1, Condition=Get bIsTargeting())
   > Set is targeting float(is targeting float:double=FInterpTo(Current=0, Target=1, DeltaTime=Get delta time x(), InterpSpeed=50)) -> (Output_Get:double)
   Set is targeting float(execute=@n9.else, is targeting float=FInterpTo(Current=0, Target=0, DeltaTime=Get delta time x(), InterpSpeed=5))
   n10 = Event OnCharacterCustomizationLoaded()
   > Set Skel Mesh(Skel Mesh=@n10.CustomBaseMesh)
  ]}
]
variables [
 crouching looking up {type=number}
 crouching moving {type=boolean}
 delta time x {type=number}
 earring rotation alpha {type=number}
 head rotating right when melee {type=number}
 is melee float {type=number}
 is targeting float {type=number}
 not sky diving {type=boolean}
 Skel Mesh {type=SkeletalMeshComponent}
 targeting up {type=number}
 targeting up and rotationg {type=number}
]
